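# Skydiving contrail Niagara system that flies three missile meshes around the player in gameplay and none in the front end.
# Missiles sit in a ring on the skeletal mesh, with waveform drift, timeline-driven catch-up offsets and dive-blended velocity orientation plus spin.
# FlameLicks, ThrusterSpikes and Light spawn from Missiles through particle attribute readers, re-reading position and orientation to pin flame sprites and lights to each one.
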NiagaraSystem /BRCosmetics/Athena/Effects/Particles/Skydiving/NS_Contrail_TalonPane/NS_Contrail_TalonPane
dependencies {
 usedDataInterfaces [
  &NiagaraDataInterfaceCurve
  &NiagaraDataInterfaceParticleRead
  &NiagaraDataInterfaceVectorCurve
  &NiagaraDataInterfaceColorCurve
 ]
 usedDynamicInputs [
  &VectorFromCurve
  &FN_FrontEndVector3Modifier
  &TransformVector
  &SetBoolByFloatComparison
  &Divide_Float
  &Abs_Float
  &OneMinusFloat
 ]
 usedModules [
  &self:SystemSpawnScript.SystemScriptSource.SystemScriptGraph.NiagaraNodeAssignment_0.SetVariables_#1
  &SystemState
  /Game/Effects/Niagara/Modules/Skydiving/FN_DivingAttributes/v5/FN_DivingAttributes
  &self:UpdatePreviousRotation
  &self:SystemSpawnScript.SystemScriptSource.SystemScriptGraph.NiagaraNodeAssignment_4.SetVariables_#2
  &self:SystemSpawnScript.SystemScriptSource.SystemScriptGraph.NiagaraNodeAssignment_1.SetVariables_#3
  &FN_Integrator
  &Timeline
  &self:SystemSpawnScript.SystemScriptSource.SystemScriptGraph.NiagaraNodeAssignment_7.SetVariables_#4
  &self:SystemSpawnScript.SystemScriptSource.SystemScriptGraph.NiagaraNodeAssignment_9.SetVariables_#5
 ]
}
emitters {
 FlameLicks {
  emitterUpdateScript [
   &EmitterState()
   &SpawnParticlesFromOtherEmitter(
    Attribute Reader=DataInterface() {propertyValues {EmitterBinding {bindingMode=Other, emitterName=Missiles}}}
    Spawn Rate=36
    Spawn Rate Per Particle Cap=8
    Use Maximum Age=false
    Use Minimum Age=false
   )
  ]
  particleSpawnScript [
   &InitializeParticle(
    Color=<rgba 1 1 1 0.4>
    Lifetime Min=1.5
    Lifetime Mode=::NewEnumerator1
    Sprite Size Mode=::NewEnumerator1
    Uniform Sprite Size Max=60
    Uniform Sprite Size Min=45
    UsePositionOffset=false
    Write Material Random=true
   )
   &SampleParticlesFromOtherEmitter(Applied Position Offset:Vector3f=<0 0 50>, Mesh Orientation Sampling:ENiagara_DirectReadSamplingMode=::NewEnumerator0)
   &CameraOffset(Camera Offset Amount=&RandomRangeFloat(Minimum=-8, Maximum=-3))
   /SetVariables(
    Particles.RandVectorEnd:Vector3f=&RandomRangeVector(Minimum=<-20 -20 0>, Maximum=<20 20 0>)
    Particles.RandVectorStart:Vector3f=&RandomRangeVector(Minimum=<-20 -20 0>, Maximum=<20 20 0>)
   ) {isSetParameters=true, module=SetVariables_#6}
  ]
  particleUpdateScript [
   &ParticleState()
   &ScaleSpriteSize(Scale Factor=<0.2 1.5>, Scale Sprite Size Mode=::NewEnumerator2) {module=ScaleSpriteSize002}
   &ScaleSpriteSize(Uniform Curve Sprite Scale=DataInterface() {propertyValues {Curve {keys [{tangentMode=RCTM_User, time=0.25, value=0}, {tangentMode=RCTM_User, time=0.5, value=1}]}}}) {module=ScaleSpriteSize001}
   &ShapeLocation(
    Apply To Particle Position=::NewEnumerator1
    MaskSpawnGroup=false
    Override Seed=false
    Rotation Mode=::NewEnumerator5
    Scale Mode:ENiagara_ScaleMode=::NewEnumerator1
    Sphere Radius=80
   ) {enabled=false}
   &InheritSourceMovement(
    Applied Position Scale=&VectorFromFloat(Value=&FloatFromCurve(FloatCurve=DataInterface() {propertyValues {Curve {keys [<at 0 1>,<at 1 0.95>]}}}))
    ApplyForce=false
    ApplyPosition=true
    ApplyPositionFalloff=false
   )
   &SolveForcesAndVelocity(Clamp Velocity=false, Limit Acceleration=false)
   &ScaleSpriteSize(Scale Factor=&Vector2DFromFloat(Value=&FloatFromCurve(FloatCurve=DataInterface() {propertyValues {Curve {keys [
          {arriveTangent=7.5764, interpMode=RCIM_Cubic, leaveTangent=7.5765, tangentMode=RCTM_User, time=0, value=0.1}
          {arriveTangent=0.1045, interpMode=RCIM_Cubic, leaveTangent=0.1045, tangentMode=RCTM_User, time=0.25, value=1}
          {arriveTangent=-0.8944, interpMode=RCIM_Cubic, leaveTangent=-0.8944, tangentMode=RCTM_User, time=1, value=0.8}
         ]}}}))) {enabled=false}
   &DynamicMaterialParameters(
    Index 0 Param 1=&RandomRangeFloat(Minimum=0, Maximum=1)
    Index 0 Param 4=1
    Index 2 Param 1=1
    Index 2 Param 2=&FloatFromCurve(FloatCurve=DataInterface() {propertyValues {Curve {defaultValue=1, keys [<at 0 1>, <at 0.25 0>, {tangentMode=RCTM_User, time=0.5, value=0}, {tangentMode=RCTM_User, time=1, value=1}]}}})
    Module.Param0WriteEnabled=true
    Module.Param0WriteEnabled2=true
    Module.Param1WriteEnabled=false
    Module.Param1WriteEnabled2=true
    Module.Param2WriteEnabled=false
    Module.Param2WriteEnabled2=true
    Module.Param3WriteEnabled=false
    Module.Param3WriteEnabled2=true
    Write Parameter Index 2=true
   )
   &ScaleColor(
    Scale Mode=::NewEnumerator1
    Scale RGBA=&Vector4FromCurve(Vector4Curve=DataInterface() {propertyValues {
       AlphaCurve {keys [<at 0.75 1>,<at 1 0>]}
       BlueCurve {keys [<at 0 1>,<at 0.1771 15>,<at 0.5441 1.2674>,<at 1 1>]}
       GreenCurve {keys [<at 0 1>,<at 0.1771 15>,<at 0.5441 1.2674>,<at 1 1>]}
       RedCurve {keys [<at 0 1>,<at 0.1771 15>,<at 0.5441 1.2674>,<at 1 1>]}
      }})
    ScaleRGBA=true
   ) {enabled=false}
   &SubUVAnimation(Start Frame:NiagaraInt32=&RandomRangeInt(Minimum=0, Maximum=20)) {enabled=false}
   &ScaleColor(
    Scale RGB=&MakeVectorFromLinearColorRGB(Color=&LerpLinearColors(
      StartColor=<rgba 0 1.3565 2 1>
      EndColor=<rgba 2 0.2648 0 1>
      LerpFactor=$System.InterpolateOverTime.InterpolateOverTime_Float.Moving Average
     ))
    ScaleA=true
    ScaleRGB=true
   ) {enabled=false, module=ScaleColor001}
   /SetVariables(Particles.Position=&AddVectorToPosition(
     Position=&self:Get_Position_Attribute_By_Index(
      Particle Attribute Reader:NiagaraDataInterfaceParticleRead=DataInterface() {propertyValues {EmitterBinding {bindingMode=Other, emitterName=Missiles}}}
      Particle Index:NiagaraInt32=&MakeIntFromFloat(Float=$Particles.SampleParticlesFromOtherEmitter.SampledIndex)
     )
     Vector=&RotateVectorByQuaternion(
      Vector To Rotate=&Lerp_Vector(
       A=$Particles.RandVectorStart
       B=&Add_Vector(A=&RandomRangeVector(Minimum=<0 0 250>, Maximum=<0 0 350>), B=$Particles.RandVectorEnd)
       Alpha=&FloatFromCurve(FloatCurve=DataInterface() {propertyValues {Curve {defaultValue=1, keys [{tangentMode=RCTM_User, time=0, value=0}, {tangentMode=RCTM_User, time=1, value=1}]}}})
      )
      Quaternion=&self:Get_MeshOrientation_Attribute_By_Index(
       Particle Attribute Reader:NiagaraDataInterfaceParticleRead=DataInterface() {propertyValues {EmitterBinding {bindingMode=Other, emitterName=Missiles}}}
       Particle Index:NiagaraInt32=&MakeIntFromFloat(Float=$Particles.SampleParticlesFromOtherEmitter.SampledIndex)
      )
     )
    )) {isSetParameters=true, module=SetVariables_#7}
   &Color(Color=&ColorFromCurve(ColorCurve=DataInterface() {propertyValues {
       AlphaCurve {keys [<at 0 1 RCIM_Cubic>,<at 1 0 RCIM_Cubic>]}
       BlueCurve {keys [<at 0 1 RCIM_Cubic>,<at 1 0.0034 RCIM_Cubic>]}
       GreenCurve {keys [<at 0 1 RCIM_Cubic>,<at 1 0.1131 RCIM_Cubic>]}
       RedCurve {keys [<at 0 1 RCIM_Cubic>,<at 1 0.6615 RCIM_Cubic>]}
      }}))
   &SpriteFacingAndAlignment(
    Alignment Coordinate Space=::World
    Sprite Alignment=&RotateVectorByQuaternion(
     Vector To Rotate=<0 0 1>
     Quaternion=&self:Get_MeshOrientation_Attribute_By_Index(
      Particle Attribute Reader:NiagaraDataInterfaceParticleRead=DataInterface() {propertyValues {EmitterBinding {bindingMode=Other, emitterName=Missiles}}}
      Particle Index:NiagaraInt32=&MakeIntFromFloat(Float=$Particles.SampleParticlesFromOtherEmitter.SampledIndex)
     )
    )
    WriteAlignment=true
    WriteFacing=false
   )
  ]
  rendererClasses [&NiagaraSpriteRendererProperties]
  renderers [
   {rendererClass=&NiagaraSpriteRendererProperties}
   {rendererClass=&NiagaraSpriteRendererProperties, rendererIndex=1}
  ]
 }
 Light {
  emitterUpdateScript [
   &EmitterState(
    Inactive Response=::NewEnumerator1
    Life Cycle Mode=::NewEnumerator1
    Loop Behavior=::NewEnumerator1
    Loop Duration Mode=::NewEnumerator1
    UseLoopDelay=false
   )
   &SpawnBurst_Instantaneous(Spawn Count=3, Use Spawn Probability=false, UseLoopCountLimit=false) {enabled=false}
   &SpawnParticlesFromOtherEmitter(
    Attribute Reader=DataInterface() {propertyValues {EmitterBinding {bindingMode=Other, emitterName=Missiles}}}
    Calculate Spawn Rate Per Particle=false
    Maximum Age Spawn Threshold=0.15
    Spawn Rate=30
    Use Maximum Age=true
    Use Minimum Age=false
   )
  ]
  particleSpawnScript [
   &PartitionParticles(Number Of Partitions=3, Partition Mode:ENiagara_PartitionMode_v2=::NewEnumerator5) {enabled=false, module=PartitionParticles001}
   &InitializeParticle(
    Color=<rgba 1 0.3293 0.0066 1>
    Lifetime=&Multiply_Float(A=&RandomRangeFloat(Minimum=0.4, Maximum=0.8), B=1)
    Mesh Scale=&Multiply_VectorByFloat(Vector=&RandomRangeVector(Minimum=<2.5 0.5 0.5>, Maximum=<5 2 2>), Float=0.35)
    Mesh Scale Mode=::NewEnumerator4
    Sprite Size Mode=::NewEnumerator1
    Uniform Sprite Size Max=60
    Uniform Sprite Size Min=45
    UsePositionOffset=false
    Write Material Random=true
   )
   &SampleParticlesFromOtherEmitter(
    Applied Position Offset:Vector3f=<0 0 50>
    Mesh Orientation Sampling:ENiagara_DirectReadSamplingMode=::NewEnumerator0
    Transform Data Option=1
   )
   &CameraOffset(Camera Offset Amount=10)
   /SetVariables(Particles.RandVectorStart:Vector3f=<0 0 30>) {isSetParameters=true, module=SetVariables_#6}
   &KillParticles(Kill Particles=&SetBoolByIntComparison(A=&ReturnParticleIDAttributes(), B=2, Comparison Type=::NewEnumerator0))
  ]
  particleUpdateScript [
   &ParticleState(Kill Particles When Lifetime Has Elapsed=false, Loop Particles Lifetime=true)
   &SolveForcesAndVelocity(Clamp Velocity=false, Limit Acceleration=false)
   /SetVariables(Particles.Position=&AddVectorToPosition(
     Position=&self:Get_Position_Attribute_By_Index(
      Particle Attribute Reader:NiagaraDataInterfaceParticleRead=DataInterface() {propertyValues {EmitterBinding {bindingMode=Other, emitterName=Missiles}}}
      Particle Index:NiagaraInt32=&MakeIntFromFloat(Float=$Particles.SampleParticlesFromOtherEmitter.SampledIndex)
     )
     Vector=&RotateVectorByQuaternion(
      Vector To Rotate=$Particles.RandVectorStart
      Quaternion=&self:Get_MeshOrientation_Attribute_By_Index(
       Particle Attribute Reader:NiagaraDataInterfaceParticleRead=DataInterface() {propertyValues {EmitterBinding {bindingMode=Other, emitterName=Missiles}}}
       Particle Index:NiagaraInt32=&MakeIntFromFloat(Float=$Particles.SampleParticlesFromOtherEmitter.SampledIndex)
      )
     )
    )) {isSetParameters=true, module=SetVariables_#7}
   &DynamicMaterialParameters(
    Index 0 Param 1=3
    Index 0 Param 2=5
    Module.Index0Float0Write=true
    Module.Index0Float1Write=true
    Module.Index0Float2Write=true
    Module.Index0Float3Write=true
   )
   &FN_Light_Attributes(
    Color=<rgba 1 0.3293 0.0066 1>
    Create Color Attribute=true
    Create Exponent Attribute=true
    Create Radius Attribute=true
    Enable Particle Light Enable=true
    Exponent=3
    Light Enabled=true
    Radius=30
    Scale Color=120
   )
  ]
  rendererClasses [&NiagaraSpriteRendererProperties, &NiagaraLightRendererProperties]
  renderers [
   {rendererClass=&NiagaraSpriteRendererProperties}
   {rendererClass=&NiagaraLightRendererProperties, rendererIndex=1}
  ]
 }
 Missiles {
  emitterUpdateScript [
   &EmitterState()
   &SpawnBurst_Instantaneous(
    Spawn Count=&MakeIntFromFloat(Float=&FN_FrontEndFloatModifier(Gameplay=3, Front End=0))
    Use Spawn Probability=false
    UseLoopCountLimit=true
   )
   &InterpolateOverTime(Rate Of Change=50, Target Vector Value:Vector3f=$Engine.Owner.Velocity, Variable Type=::NewEnumerator2)
  ]
  particleSpawnScript [
   &FN_SetSpawnGroup(Number of Spawn Groups=3)
   &InitializeParticle(
    Lifetime=2
    Mesh Scale Mode=::NewEnumerator3
    Mesh Uniform Scale=$System.Ship Scale
    Sprite Size Mode=::NewEnumerator3
    Uniform Sprite Size=40
    UsePositionOffset=true
    Write Material Random=true
   )
  ]
  particleUpdateScript [
   &ParticleState(Kill Particles When Lifetime Has Elapsed=false, Let Infinitely Lived Particles Die When Emitter Deactivates=true)
   /SetVariables(Particles.Ambient:Vector3f=&Multiply_VectorByFloat(
     Vector=&MakeVector(
      X=&Waveform(
       "Waveform [1]"=::NewEnumerator12
       "[1] Input Position"=&RandomRangeFloat(Minimum=0, Maximum=1000)
       "[1] Frequency"=1
       "[1] Amplitude Min/Max"=<-50 50>
      )
      Y=&Waveform(
       "Waveform [1]"=::NewEnumerator12
       "[1] Input Position"=&RandomRangeFloat(Minimum=1200, Maximum=1800)
       "[1] Frequency"=1
       "[1] Amplitude Min/Max"=<-50 50>
      )
      Z=&Waveform(
       "Waveform [1]"=::NewEnumerator12
       "[1] Input Position"=&RandomRangeFloat(Minimum=2200, Maximum=2800)
       "[1] Frequency"=1
       "[1] Amplitude Min/Max"=<-50 50>
      )
     )
     Float=0.1
    )) {isSetParameters=true}
   &InterpolateOverTime(
    Rate Of Change=3
    Target Float Value:NiagaraFloat=&MakeCustomFloatFromBool(Bool=$System.Timeline001.IsRewinding, True Float=0, False Float=1)
   ) {module=InterpolateOverTime004}
   /SetVariables(Particles.CatchupVector:Vector3f=&Multiply_VectorByFloat(
     Vector=&Multiply_VectorByFloat(
      Vector=&NormalizeVector(Vector To Normalize=$System.SkydiveVelocity, Safe Normalize=true, Safe Fallback Threshold=0, Safe Fallback Vector=<0 0 0.001>)
      Float=-100
     )
     Float=&Lerp_Float(
      A=&FloatFromCurve(
       FloatCurve=DataInterface() {propertyValues {Curve {keys [
           {interpMode=RCIM_Cubic, tangentMode=RCTM_User, time=0, value=0}
           {interpMode=RCIM_Cubic, tangentMode=RCTM_User, time=0.35, value=-0.35}
           {tangentMode=RCTM_User, time=1, value=0}
          ]}}}
       CurveIndex=$System.Timeline001.NormalizedTime
      )
      B=&FloatFromCurve(
       FloatCurve=DataInterface() {propertyValues {Curve {keys [
           {arriveTangent=3.3827, arriveTangentWeight=1.1106, interpMode=RCIM_Cubic, leaveTangentWeight=1.1106, time=0, value=0}
           {arriveTangentWeight=0.5241, interpMode=RCIM_Cubic, leaveTangentWeight=0.1667, time=0.5, value=0.65}
           <at 1 0 RCIM_Cubic>
          ]}}}
       CurveIndex=$System.Timeline001.NormalizedTime
      )
      Alpha=$Particles.InterpolateOverTime004.InterpolateOverTime_Float.Moving Average
     )
    )) {isSetParameters=true}
   &ShapeLocation(
    Apply To Particle Position=::NewEnumerator1
    MaskSpawnGroup=false
    Offset=&Lerp_Vector(A=<0 0 -150>, B=<0 0 0>, Alpha=$System.InterpolateOverTime002.InterpolateOverTime_Float.Moving Average)
    Offset Mode=::NewEnumerator0
    Override Seed=false
    Ring / Disc Distribution Mode:ENiagaraSphereDistributionMode=::NewEnumerator1
    Ring Radius=&Lerp_Float(A=150, B=220, Alpha=$System.InterpolateOverTime.InterpolateOverTime_Float.Moving Average)
    Rotation Angle=0.75
    Rotation Axis=<0 1 0>
    Rotation Coordinate Space=::World
    Rotation Mode=::NewEnumerator1
    Shape Primitive=::NewEnumerator4
    U Position=&ReturnNormalizedExecIndex(Normalized Index Includes One=false)
    UseOwnerScale=true
   )
   &SkeletalMeshLocation(
    Bone Sampling Mode=::NewEnumerator4
    MaskSpawnGroup=false
    Orientation Sampling=::NewEnumerator1
    Sampled Position Offset=&Add_Vector(
     A=$Particles.CatchupVector
     B=/Niagara/DynamicInputs/Vectors/RotateVector(
      Vector To Rotate:Vector3f=&Add_Vector(
       A=$Particles.Ambient
       B=&Add_Vector(
        A=&Add_Vector(
         A=$Particles.ShapeLocation.ShapeVector
         B=&MaskVectorBySpawnGroup(
          Masked Vector=&Lerp_Vector(A=<0 110 -150>, B=<0 110 -150>, Alpha=$System.InterpolateOverTime.InterpolateOverTime_Float.Moving Average)
          Passthrough Vector=<0 0 0>
         )
        )
        B=$System.OverallTargetOffset
       )
      )
      Yaw:NiagaraFloat=&Waveform(
       "Waveform [1]"=::NewEnumerator12
       "[1] Input Position"=&RandomRangeFloat(Minimum=80000, Maximum=40000)
       "[1] Frequency"=1.25
       "[1] Amplitude Min/Max"=<7 -7>
      )
      Pitch:NiagaraFloat=0
      Roll:NiagaraFloat=0
      Rotation Rate:NiagaraFloat=0.5
     )
    )
    Skeletal Mesh=$User.SkeletalMesh
    Use Sampled Mesh Attribute as Ribbon ID=false
   )
   &ScaleMeshSize(Scale Factor=&VectorFromFloat(Value=&FloatFromCurve(
      FloatCurve=DataInterface() {propertyValues {Curve {keys [{arriveTangent=1.0186, interpMode=RCIM_Cubic, leaveTangent=1.0186, tangentMode=RCTM_User, time=0, value=0}, <at 1 1 RCIM_Cubic>]}}}
      CurveIndex=$Emitter.Age
     )))
   /SetVariables(Particles.Wobble:Quat4f=&MultiplyQuaternion(
     Quaternion A=&FindQuatBetweenTwoVectors(
      From Vector=&MakeVector(
       X=&Sine(
        Normalized Angle=&Add_Float(A=$System.FN_Integrator.Value, B=&RandomRangeFloat(Minimum=0, Maximum=1))
        Period=&MakeCustomFloatFromBool(Bool=&RandomBool(), True Float=1, False Float=-1)
        Scale=0.5
        Bias=0
       )
       Y=&Cosine(
        Normalized Angle=&Add_Float(A=$System.FN_Integrator.Value, B=&RandomRangeFloat(Minimum=0, Maximum=1))
        Period=&MakeCustomFloatFromBool(Bool=&RandomBool(), True Float=1, False Float=-1)
        Scale=0.5
        Bias=0
       )
       Z=&RandomRangeFloat(Minimum=6, Maximum=8)
      )
      To Vector=<0 0 1>
     )
     Quaternion B=$Particles.Initial.MeshOrientation
    )) {isSetParameters=true}
   &UpdateMeshOrientation(
    Bank Rate:NiagaraFloat=&Lerp_Float(
     A=&Lerp_Float(
      A=1
      B=30
      Alpha=&Waveform(
       "Waveform [1]"=::NewEnumerator12
       "[1] Input Position"=&RandomRangeFloat(Minimum=0, Maximum=1000)
       "[1] Phase Scale"=50000
       "[1] Frequency"=0.5
       "[1] Amplitude Min/Max"=<0 1>
      )
     )
     B=900
     Alpha=$System.Dive
    )
    Banking Angle Min / Max:Vector2f=&Lerp_Vector2(A=<-10 10>, B=<-4 4>, Alpha=$System.Dive)
    "Foward Vector (Mesh Space)":Vector3f=&Lerp_Vector(A=<0.5 0 1>, B=<1 0 0>, Alpha=$System.InterpolateOverTime.InterpolateOverTime_Float.Moving Average)
    Orientation Method=::NewEnumerator3
    Particle Velocity:Vector3f=$Emitter.InterpolateOverTime.InterpolateOverTime_V3_.Moving Average
    Rotation Decay Rate In Degrees:NiagaraFloat=&Lerp_Float(A=150, B=500, Alpha=$System.Dive)
    Up Vector:Vector3f=&Lerp_Vector(
     A=&ReturnMeshOrientationAxis(Mesh Orientation Axis=::YAxis, Mesh Orientation Quat=$Particles.SkeletalMeshLocation.SampledRotation)
     B=&Multiply_VectorByFloat(
      Vector=&ReturnMeshOrientationAxis(Mesh Orientation Axis=::ZAxis, Mesh Orientation Quat=$Particles.SkeletalMeshLocation.SampledRotation)
      Float=1
     )
     Alpha=&FloatFromCurve(FloatCurve=DataInterface() {propertyValues {Curve {keys [<at 0 0.5>,<at 0.25 0.75>]}}}, CurveIndex=$System.Dive)
    )
    Upright Roll Correction Speed:NiagaraFloat=&Lerp_Float(A=25, B=90000, Alpha=$System.Dive)
    Velocity Orientation Convergence Rate:NiagaraFloat=&Lerp_Float(
     A=&Lerp_Float(
      A=3
      B=12
      Alpha=&Waveform(
       "Waveform [1]"=::NewEnumerator12
       "[1] Input Position"=&RandomRangeFloat(Minimum=0, Maximum=100)
       "[1] Frequency"=5
       "[1] Amplitude Min/Max"=<0 1>
      )
     )
     B=&Lerp_Float(
      A=20
      B=40
      Alpha=&Waveform(
       "Waveform [1]"=::NewEnumerator12
       "[1] Input Position"=&RandomRangeFloat(Minimum=0, Maximum=100)
       "[1] Frequency"=5
       "[1] Amplitude Min/Max"=<0 1>
      )
     )
     Alpha=$System.Dive
    )
   ) {enabled=false}
   /BRCosmetics/Athena/Effects/Particles/Skydiving/NS_Contrail_RedPepper/UpdateMeshOrientation_RedPepper_TEMP(
    Bank Rate:NiagaraFloat=&Lerp_Float(
     A=&Lerp_Float(
      A=1
      B=30
      Alpha=&Waveform(
       "Waveform [1]"=::NewEnumerator12
       "[1] Input Position"=&RandomRangeFloat(Minimum=0, Maximum=1000)
       "[1] Phase Scale"=50000
       "[1] Frequency"=0.5
       "[1] Amplitude Min/Max"=<0 1>
      )
     )
     B=900
     Alpha=$System.Dive
    )
    Banking Angle Min / Max:Vector2f=&Lerp_Vector2(A=<-10 10>, B=<-4 4>, Alpha=$System.Dive)
    "Foward Vector (Mesh Space)":Vector3f=<0.25 0 1>
    Orientation Method:ENiagara_UpdateMeshOrientationMode=::NewEnumerator3
    Particle Velocity:Vector3f=$Emitter.InterpolateOverTime.InterpolateOverTime_V3_.Moving Average
    Rotation Decay Rate In Degrees:NiagaraFloat=&Lerp_Float(A=150, B=500, Alpha=$System.Dive)
    Up Vector:Vector3f=&Lerp_Vector(
     A=&ReturnMeshOrientationAxis(Mesh Orientation Axis=::YAxis, Mesh Orientation Quat=$Particles.SkeletalMeshLocation.SampledRotation)
     B=&Multiply_VectorByFloat(
      Vector=&ReturnMeshOrientationAxis(Mesh Orientation Axis=::ZAxis, Mesh Orientation Quat=$Particles.SkeletalMeshLocation.SampledRotation)
      Float=1
     )
     Alpha=&FloatFromCurve(FloatCurve=DataInterface() {propertyValues {Curve {keys [<at 0 0.5>,<at 0.25 0.75>]}}}, CurveIndex=$System.Dive)
    )
    Upright Roll Correction Speed:NiagaraFloat=&Lerp_Float(A=25, B=90000, Alpha=$System.Dive)
    Velocity Orientation Convergence Rate:NiagaraFloat=&Lerp_Float(
     A=&Lerp_Float(
      A=3
      B=12
      Alpha=&Waveform(
       "Waveform [1]"=::NewEnumerator12
       "[1] Input Position"=&RandomRangeFloat(Minimum=0, Maximum=100)
       "[1] Frequency"=5
       "[1] Amplitude Min/Max"=<0 1>
      )
     )
     B=&Lerp_Float(
      A=20
      B=40
      Alpha=&Waveform(
       "Waveform [1]"=::NewEnumerator12
       "[1] Input Position"=&RandomRangeFloat(Minimum=0, Maximum=100)
       "[1] Frequency"=5
       "[1] Amplitude Min/Max"=<0 1>
      )
     )
     Alpha=$System.Dive
    )
   )
   &InterpolateOverTime(Rate Of Change=5, Target Quat Value:Quat4f=$Particles.MeshOrientation, Variable Type=::NewEnumerator5) {module=InterpolateOverTime002}
   /SetVariables(Particles.New Orient:Quat4f=$Particles.InterpolateOverTime002.InterpolateOverTime_Quaternion.Moving Average) {isSetParameters=true}
   /SetVariables(Particles.MeshOrientation=&MultiplyQuaternion(
     Quaternion A=$Particles.New Orient
     Quaternion B=&MakeQuaternion(
      Quaternion From=0
      Axis=<0 0 1>
      Angle=&FloatFromCurve(
       FloatCurve=DataInterface() {propertyValues {
         bUseLUT=false
         Curve {
          defaultValue=0
          postInfinityExtrap=RCCE_CycleWithOffset
          keys [{tangentMode=RCTM_User, time=0, value=0}, {tangentMode=RCTM_User, time=1, value=1}]
         }
        }}
       CurveIndex=$Emitter.Age
       Scale Curve=120
      )
      From=?
      To=?
      Angle Type=::NewEnumerator0
      Normalize Vectors=false
     )
    )) {isSetParameters=true}
   /SetVariables(Particles.MeshOrientation=$Particles.New Orient) {enabled=false, isSetParameters=true}
   &ScaleColor(Scale RGB=&MakeVectorFromLinearColorRGB(Color=<rgba 1 1 1 1>), ScaleA=true, ScaleRGB=true)
   &DynamicMaterialParameters(
    Index 0 Param 1=1
    Index 0 Param 2=0.5
    Index 0 Param 3=&RandomRangeFloat(Minimum=0.1, Maximum=1)
    Index 0 Param 4=&RandomRangeFloat(Minimum=0.75, Maximum=1)
    Module.Param0WriteEnabled=true
    Module.Param1WriteEnabled=true
    Module.Param2WriteEnabled=true
    Module.Param3WriteEnabled=true
   )
   /SetVariables(
    Particles.Decal Color:LinearColor=&LerpLinearColors(
     StartColor=<rgba 0 1.3565 2 3>
     EndColor=<rgba 6 0.9039 0 3>
     LerpFactor=$System.InterpolateOverTime.InterpolateOverTime_Float.Moving Average
    )
    Particles.Decal Offset=15
    Particles.Decal Position:NiagaraPosition=&AddVectorToPosition(
     Position=$Particles.Position
     Vector=&RotateVectorByQuaternion(Vector To Rotate=<-10 0 15>, Quaternion=$Particles.MeshOrientation)
    )
   ) {isSetParameters=true}
   &FN_Light_Attributes(
    Color=&LerpLinearColors(
     StartColor=<rgba 0 1.3565 2 1>
     EndColor=<rgba 6 0.9039 0 4>
     LerpFactor=$System.InterpolateOverTime.InterpolateOverTime_Float.Moving Average
    )
    Create Color Attribute=true
    Create Exponent Attribute=true
    Create Radius Attribute=true
    Exponent=2
    Radius=85
    Scale Color=&FloatFromCurve(
     FloatCurve=DataInterface() {propertyValues {Curve {keys [
         <at 0 1 RCIM_Cubic>
         {arriveTangent=1.794, interpMode=RCIM_Cubic, leaveTangent=1.794, time=0.0746, value=0.7914}
         {arriveTangent=-0.0227, interpMode=RCIM_Cubic, leaveTangent=-0.0227, tangentMode=RCTM_User, time=0.121, value=1.2171}
         {arriveTangent=-5.5976, interpMode=RCIM_Cubic, leaveTangent=-5.5976, time=0.1377, value=0.3589}
         {arriveTangent=5.9152, interpMode=RCIM_Cubic, leaveTangent=5.9152, time=0.1738, value=0.9213}
         {arriveTangent=3.6344, interpMode=RCIM_Cubic, leaveTangent=3.6344, time=0.2093, value=0.7825}
         {arriveTangent=-4.8832, interpMode=RCIM_Cubic, leaveTangent=-4.8832, time=0.2396, value=1.1602}
         {arriveTangent=-6.1936, interpMode=RCIM_Cubic, leaveTangent=-6.1936, time=0.254, value=0.5643}
         {arriveTangent=-3.4254, interpMode=RCIM_Cubic, leaveTangent=-3.4254, time=0.2791, value=0.9154}
         {arriveTangent=1.2156, interpMode=RCIM_Cubic, leaveTangent=1.2156, time=0.3019, value=0.4002}
         {arriveTangent=5.6358, interpMode=RCIM_Cubic, leaveTangent=5.6358, time=0.3406, value=0.9901}
         {arriveTangent=6.3907, interpMode=RCIM_Cubic, leaveTangent=6.3907, time=0.3796, value=0.8384}
         {arriveTangent=-3.0158, interpMode=RCIM_Cubic, leaveTangent=-3.0158, time=0.4112, value=1.4411}
         {arriveTangent=-10.4214, interpMode=RCIM_Cubic, leaveTangent=-10.4214, time=0.4339, value=0.6747}
         {arriveTangent=-3.4821, interpMode=RCIM_Cubic, leaveTangent=-3.4821, time=0.4676, value=0.8529}
         {arriveTangent=-1.4741, interpMode=RCIM_Cubic, leaveTangent=-1.4741, time=0.4906, value=0.4775}
         {arriveTangent=-2.6963, interpMode=RCIM_Cubic, leaveTangent=-2.6963, time=0.5283, value=0.7634}
         {arriveTangent=2.8896, interpMode=RCIM_Cubic, leaveTangent=2.8896, time=0.5517, value=0.3126}
         {arriveTangent=6.2844, interpMode=RCIM_Cubic, leaveTangent=6.2844, time=0.5824, value=0.9196}
         {arriveTangent=1.4786, interpMode=RCIM_Cubic, leaveTangent=1.4786, time=0.5976, value=0.6006}
         {arriveTangent=4.8152, interpMode=RCIM_Cubic, leaveTangent=4.8152, time=0.6363, value=0.9993}
         {arriveTangent=2.8106, interpMode=RCIM_Cubic, leaveTangent=2.8106, time=0.6464, value=0.8356}
         {arriveTangent=3.2181, interpMode=RCIM_Cubic, leaveTangent=3.2181, time=0.6964, value=1.1681}
         {arriveTangent=1.8672, interpMode=RCIM_Cubic, leaveTangent=1.8672, time=0.7177, value=1.065}
         {arriveTangent=-12.2147, interpMode=RCIM_Cubic, leaveTangent=-12.2147, time=0.7385, value=1.2469}
         {arriveTangent=-0.0928, interpMode=RCIM_Cubic, leaveTangent=-0.0928, tangentMode=RCTM_User, time=0.7564, value=0.5916}
         {arriveTangent=3.085, interpMode=RCIM_Cubic, leaveTangent=3.085, time=0.7931, value=1.02}
         {arriveTangent=2.1222, interpMode=RCIM_Cubic, leaveTangent=2.1222, time=0.8092, value=0.7545}
         {arriveTangent=-0.424, interpMode=RCIM_Cubic, leaveTangent=-0.424, time=0.8924, value=1.2306}
         {arriveTangent=-8.511, interpMode=RCIM_Cubic, leaveTangent=-8.511, time=0.911, value=0.7114}
         {arriveTangent=3.2422, interpMode=RCIM_Cubic, leaveTangent=3.2422, time=0.9278, value=0.9291}
         {interpMode=RCIM_Cubic, leaveTangent=0.0463, time=1, value=1}
        ]}}}
     CurveIndex=$Emitter.NormalizedLoopAge
     Scale Curve=10
    )
   )
   /SetVariables(Particles.FlamesCameraOffset=5) {isSetParameters=true}
   /SetVariables(
    Particles.GlowColor=<rgba 0.4688 0 0 1>
    Particles.GlowDynamicParam2:Vector4f=<0.65 0 0 0>
    Particles.GlowPosition2:NiagaraPosition=$Particles.Position
    Particles.GlowScale2:Vector3f=&Multiply_Vector(A=$Particles.Scale, B=<1.2 1.5 1.2>)
   ) {isSetParameters=true}
   &FN_Sprite_Attributes(
    Camera Offset=10
    Color=&MultiplyLinearColor(
     Linear Color=<rgba 1 0.3223 0 1>
     Scale Factor=&MakeVector4FromVectorAndFloat("Vector (RGB)"=<1 1 1>, "Float (Alpha)"=1)
    )
    Create Camera Offset Attribute=true
    Create Color Attribute=true
    Create Position Attribute=true
    Position:NiagaraPosition=&AddVectorToPosition(Position=$Particles.Position, Vector=<0 6 28>)
    Write to Module Namespace=true
   )
   /SetVariables(Particles.AirglowDyn:Vector4f=<3 5 0 0>) {isSetParameters=true}
   /SetVariables(
    Particles.TrailDyn1:Vector4f=&MakeVector4(X=15, Y=120, Z=1, W=&RandomRangeFloat(Minimum=0, Maximum=1))
    Particles.TrailDyn2:Vector4f=&MakeVector4(X=0, Y=0, Z=0, W=1)
    Particles.TrailDyn3:Vector4f=&MakeVector4(X=1, Y=150, Z=0, W=0)
    Particles.TrailDyn4:Vector4f=&MakeVector4(X=0, Y=0, Z=0, W=0)
   ) {isSetParameters=true}
   /SetVariables(Particles.TrailColor:LinearColor=&FNTimeOfDayColor(
     Night Color=<rgba 0.01 0.01 0.017 1>
     Morning Color=<rgba 0.05 0.034 0.0326 1>
     Day Color=<rgba 0.05 0.034 0.0326 1>
     Evening Color=<rgba 0.05 0.034 0.0326 1>
    )) {isSetParameters=true}
   /SetVariables(Particles.SmokeMeshOrientation:Quat4f=&MultiplyQuaternion(
     Quaternion A=$Particles.MeshOrientation
     Quaternion B=&MakeQuaternion(
      Quaternion From=0
      Axis=<0 0 1>
      Angle=&RandomRangeFloat(Minimum=0, Maximum=360)
      From=?
      To=?
      Angle Type=::NewEnumerator0
      Normalize Vectors=false
     )
    )) {isSetParameters=true}
   /SetVariables(
    Particles.FireDynamic:Vector4f=&MakeVector4(X=&RandomRangeFloat(Minimum=0, Maximum=1), Y=0, Z=0, W=0)
    Particles.FireDynamic3:Vector4f=&MakeVector4(X=1, Y=0, Z=&Lerp_Float(A=0, B=1, Alpha=$System.Timeline001.NormalizedTime), W=0)
   ) {isSetParameters=true}
   &FN_Mesh_Attributes(
    Color=&MakeLinearColorFromVectorAndFloat(
     "Vector (RGB)"=&MakeVectorFromLinearColorRGB(Color=&Multiply_LinearColorByFloat(Linear Color=<rgba 1 0.4224 0.0753 1>, Float=3))
     "Float (Alpha)"=&FloatFromCurve(FloatCurve=DataInterface() {propertyValues {Curve {keys [<at 0 0>,<at 1 1>]}}})
    )
    Create Color Attribute=true
    Create Position Attribute=true
    Create Scale Attribute=true
    Position:Vector3f=&AddVectorToPosition(Position=$Particles.Position, Vector=<0 0 10>)
    Scale=&Multiply_VectorByFloat(Vector=<0.5 0.5 1.5>, Float=0.5)
    Write to Module Namespace=true
   )
   &FN_Mesh_Attributes(
    Color=&MakeLinearColorFromVectorAndFloat(
     "Vector (RGB)"=&MakeVectorFromLinearColorRGB(Color=&Multiply_LinearColorByFloat(Linear Color=<rgba 1 0.4224 0.0753 1>, Float=3))
     "Float (Alpha)"=&FloatFromCurve(FloatCurve=DataInterface() {propertyValues {Curve {keys [<at 0 0>,<at 1 1>]}}}, Scale Curve=0.5)
    )
    Create Color Attribute=true
    Create Position Attribute=true
    Create Scale Attribute=true
    Position:Vector3f=&AddVectorToPosition(Position=$Particles.Position, Vector=<0 0 10>)
    Scale=&Multiply_VectorByFloat(Vector=<0.5 0.5 1.5>, Float=0.5)
    Write to Module Namespace=true
   ) {module=FN_Mesh_Attributes001}
   /SetVariables(Particles.ExhaustPosition:NiagaraPosition=&AddVectorToPosition(Position=$Particles.Position, Vector=<0 0 0>)) {isSetParameters=true}
  ]
  rendererClasses [&NiagaraMeshRendererProperties]
  renderers [
   {rendererClass=&NiagaraMeshRendererProperties, rendererIndex=0}
   {rendererClass=&NiagaraMeshRendererProperties, rendererIndex=1}
   {rendererClass=&NiagaraMeshRendererProperties, rendererIndex=2}
   {rendererClass=&NiagaraMeshRendererProperties, rendererIndex=3}
   {rendererClass=&NiagaraMeshRendererProperties, rendererIndex=4}
  ]
 }
 ThrusterSpikes {
  emitterUpdateScript [
   &EmitterState()
   &SpawnParticlesFromOtherEmitter(
    Attribute Reader=DataInterface() {propertyValues {EmitterBinding {bindingMode=Other, emitterName=Missiles}}}
    Spawn Rate=20
    Use Maximum Age=false
    Use Minimum Age=false
   )
  ]
  particleSpawnScript [
   &PartitionParticles(Number Of Partitions=3, Partition Mode:ENiagara_PartitionMode_v2=::NewEnumerator5) {enabled=false, module=PartitionParticles001}
   &InitializeParticle(
    Lifetime=&Multiply_Float(A=&RandomRangeFloat(Minimum=0.4, Maximum=0.8), B=1)
    Mesh Scale=&Multiply_VectorByFloat(Vector=&RandomRangeVector(Minimum=<2.5 0.5 0.5>, Maximum=<5 2 2>), Float=0.35)
    Mesh Scale Mode=::NewEnumerator4
    Sprite Size Mode=::NewEnumerator1
    Uniform Sprite Size Max=60
    Uniform Sprite Size Min=45
    UsePositionOffset=false
    Write Material Random=true
   )
   &SampleParticlesFromOtherEmitter(Applied Position Offset:Vector3f=<0 0 50>, Mesh Orientation Sampling:ENiagara_DirectReadSamplingMode=::NewEnumerator0)
   &CameraOffset(Camera Offset Amount=&RandomRangeFloat(Minimum=-8, Maximum=-3))
   /SetVariables(Particles.RandVectorEnd:Vector3f=<0 0 70>, Particles.RandVectorStart:Vector3f=<0 0 10>) {isSetParameters=true, module=SetVariables_#6}
  ]
  particleUpdateScript [
   &ParticleState()
   &ScaleSpriteSize(
    Scale Factor=&Multiply_Vector2D_ByFloat(Float=1.2, Vector2D=&RandomRangeVector2D(Minimum=<0.3 1>, Maximum=<0.45 1.2>))
    Scale Sprite Size Mode=::NewEnumerator2
   ) {module=ScaleSpriteSize002}
   &ScaleSpriteSize(Uniform Curve Sprite Scale=DataInterface() {propertyValues {Curve {
       defaultValue=1
       keys [
        {arriveTangent=1.454, interpMode=RCIM_Cubic, leaveTangent=1.454, tangentMode=RCTM_User, time=0, value=0}
        {interpMode=RCIM_Cubic, tangentMode=RCTM_User, time=1, value=1}
       ]
      }}}) {module=ScaleSpriteSize001}
   &InheritSourceMovement(
    Applied Position Scale=&VectorFromFloat(Value=&FloatFromCurve(FloatCurve=DataInterface() {propertyValues {Curve {keys [<at 0 1>,<at 1 0.95>]}}}))
    ApplyForce=false
    ApplyPosition=true
    ApplyPositionFalloff=false
   )
   &SolveForcesAndVelocity(Clamp Velocity=false, Limit Acceleration=false)
   /SetVariables(Particles.Position=&AddVectorToPosition(
     Position=&self:Get_Position_Attribute_By_Index(
      Particle Attribute Reader:NiagaraDataInterfaceParticleRead=DataInterface() {propertyValues {EmitterBinding {bindingMode=Other, emitterName=Missiles}}}
      Particle Index:NiagaraInt32=&MakeIntFromFloat(Float=$Particles.SampleParticlesFromOtherEmitter.SampledIndex)
     )
     Vector=&RotateVectorByQuaternion(
      Vector To Rotate=&Lerp_Vector(
       A=$Particles.RandVectorStart
       B=$Particles.RandVectorEnd
       Alpha=&FloatFromCurve(FloatCurve=DataInterface() {propertyValues {Curve {defaultValue=0, keys [{tangentMode=RCTM_User, time=0, value=0}, {tangentMode=RCTM_User, time=1, value=1}]}}})
      )
      Quaternion=&self:Get_MeshOrientation_Attribute_By_Index(
       Particle Attribute Reader:NiagaraDataInterfaceParticleRead=DataInterface() {propertyValues {EmitterBinding {bindingMode=Other, emitterName=Missiles}}}
       Particle Index:NiagaraInt32=&MakeIntFromFloat(Float=$Particles.SampleParticlesFromOtherEmitter.SampledIndex)
      )
     )
    )) {isSetParameters=true, module=SetVariables_#7}
   &SpriteFacingAndAlignment(
    Alignment Coordinate Space=::World
    Sprite Alignment=&RotateVectorByQuaternion(
     Vector To Rotate=<0 0 1>
     Quaternion=&self:Get_MeshOrientation_Attribute_By_Index(
      Particle Attribute Reader:NiagaraDataInterfaceParticleRead=DataInterface() {propertyValues {EmitterBinding {bindingMode=Other, emitterName=Missiles}}}
      Particle Index:NiagaraInt32=&MakeIntFromFloat(Float=$Particles.SampleParticlesFromOtherEmitter.SampledIndex)
     )
    )
    WriteAlignment=true
    WriteFacing=false
   )
   &ScaleColor(
    Scale Alpha=&Multiply_Float(
     A=&FloatFromCurve(FloatCurve=DataInterface() {propertyValues {Curve {keys [
          {arriveTangent=-0.0753, interpMode=RCIM_Cubic, leaveTangent=-0.0753, tangentMode=RCTM_User, time=0, value=1}
          {arriveTangent=-1.0655, interpMode=RCIM_Cubic, leaveTangent=-1.0655, tangentMode=RCTM_User, time=0.6798, value=0.7786}
          {arriveTangent=-0.0405, interpMode=RCIM_Cubic, leaveTangent=-0.0405, tangentMode=RCTM_User, time=1, value=0}
         ]}}})
     B=2
    )
    Scale RGB=&Multiply_VectorByFloat(
     Vector=&MakeVectorFromLinearColorRGB(Color=&ColorFromCurve(ColorCurve=DataInterface() {propertyValues {
         AlphaCurve {keys [<at 0 1 RCIM_Cubic>,<at 1 1 RCIM_Cubic>]}
         BlueCurve {keys [<at 0.01 0>, <at 0.2688 0>, <at 0.5951 0 RCIM_Cubic>, {interpMode=RCIM_Cubic, leaveTangent=0.1755, time=0.9397, value=0}]}
         GreenCurve {keys [
           <at 0.01 1>
           <at 0.2688 0.4793>
           {arriveTangent=-0.6593, interpMode=RCIM_Cubic, leaveTangent=-0.6593, time=0.5951, value=0.2667}
           {interpMode=RCIM_Cubic, leaveTangent=-0.7285, time=0.9397, value=0.037}
          ]}
         RedCurve {keys [
           <at 0.01 1>
           <at 0.2688 1>
           {arriveTangent=-0.8288, interpMode=RCIM_Cubic, leaveTangent=-0.8288, time=0.5951, value=1}
           {interpMode=RCIM_Cubic, leaveTangent=-3.1598, time=0.9397, value=0.444}
          ]}
        }}))
     Float=&Multiply_Float(
      A=&Lerp_Float(
       A=1.5
       B=1
       Alpha=&FloatFromCurve(FloatCurve=DataInterface() {propertyValues {Curve {keys [
            {arriveTangent=0.0028, interpMode=RCIM_Cubic, leaveTangent=0.0028, tangentMode=RCTM_User, time=0, value=1}
            {arriveTangent=-2.5249, interpMode=RCIM_Cubic, leaveTangent=-2.5249, tangentMode=RCTM_User, time=0.1541, value=0.8733}
            {arriveTangent=-0.0405, interpMode=RCIM_Cubic, leaveTangent=-0.0405, tangentMode=RCTM_User, time=0.8, value=0}
           ]}}})
      )
      B=&FNTimeOfDayScalar(Night Scale=1, Morning Scale=1, Evening Scale=1)
     )
    )
    ScaleA=true
    ScaleRGB=true
   ) {module=ScaleColor003}
  ]
  rendererClasses [&NiagaraSpriteRendererProperties]
  renderers [{rendererClass=&NiagaraSpriteRendererProperties}]
 }
}
summary {
 systemName=NS_Contrail_TalonPane
 emitters [
  {emitterName=Missiles, rendererClasses [&NiagaraMeshRendererProperties]}
  {emitterName=FlameLicks, rendererClasses [&NiagaraSpriteRendererProperties]}
  {emitterName=ThrusterSpikes, rendererClasses [&NiagaraSpriteRendererProperties]}
  {emitterName=Light, rendererClasses [&NiagaraSpriteRendererProperties, &NiagaraLightRendererProperties]}
 ]
 userVariables [
  User.bIsFrontEnd {type=NiagaraBool}
  User.bIsFrontEndPreview {type=NiagaraBool}
  User.Color {type=LinearColor, description="", defaultValue {struct=LinearColor, value=<rgba 1 1 1 1>}}
  User.GlobalScaler {type=NiagaraFloat, defaultValue=0, description=""}
  User.ParaGlideLeanAlpha {type=NiagaraFloat, defaultValue=0, description=""}
  User.RibbonIDOffset {type=NiagaraInt32}
  User.SkeletalMesh {
   type=NiagaraDataInterfaceSkeletalMesh
   defaultValue {
    struct=NiagaraExt_VariableValue_DataInterface
    value {dataInterface=&self:NiagaraDataInterfaceSkeletalMesh_2}
   }
  }
 ]
}
