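# Niagara skydiving contrail for the TalonPane cosmetic that trails firework flares with sparks and dust.
# Flare_With_Streak spawns flares with a random pick from three colors; Dust and FireworkFlare sprite emitters spawn from them via particle attribute readers, copying that pick for color.
# _FE twins add front-end position, size and camera offsets, while FrontEnd modifiers switch rates and scales.
# System timeline and time-of-day values drive spawn rate, movement inheritance and brightness.
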NiagaraSystem /BRCosmetics/Athena/Effects/Particles/Skydiving/NS_Contrail_TalonPane/NS_Contrail_TalonPane_Fireworks
dependencies {
 usedDataInterfaces [
  &NiagaraDataInterfaceParticleRead
  &NiagaraDataInterfaceCurve
  &NiagaraDataInterfaceArrayInt32
  &NiagaraDataInterfaceArrayColor
  &NiagaraDataInterfaceArrayFloat
 ]
 usedDynamicInputs [
  &MakeVector2D
  &SetBoolByFloatComparison
  &FNTimeOfDayScalar
 ]
 usedModules [
  &SystemState
  &Timeline
  &self:SystemSpawnScript.SystemScriptSource.SystemScriptGraph.NiagaraNodeAssignment_1.SetVariables_#1
  &self:SystemSpawnScript.SystemScriptSource.SystemScriptGraph.NiagaraNodeAssignment_2.SetVariables_#2
  &self:SystemSpawnScript.SystemScriptSource.SystemScriptGraph.NiagaraNodeAssignment_0.SetVariables_#3
  &self:SystemSpawnScript.SystemScriptSource.SystemScriptGraph.NiagaraNodeAssignment_3.SetVariables_#4
 ]
}
emitters {
 Dust {
  emitterUpdateScript [
   &EmitterState()
   &SpawnBurst_Instantaneous(Spawn Count=8, Use Spawn Probability=false, UseLoopCountLimit=false) {enabled=false}
   &SpawnParticlesFromOtherEmitter(
    Attribute Reader=DataInterface() {propertyValues {EmitterBinding {bindingMode=Other, emitterName=Flare_With_Streak}}}
    Maximum Age Spawn Threshold=0
    Spawn Rate=&FN_FrontEndFloatModifier(Gameplay=250, Front End=0)
    Use Maximum Age=false
    Use Minimum Age=false
   )
  ]
  particleSpawnScript [
   &InitializeParticle(
    Color Mode=::NewEnumerator0
    Lifetime Max=2.6
    Lifetime Min=0.7
    Lifetime Mode=::NewEnumerator1
    Position Mode=::NewEnumerator0
    Sprite Rotation Mode=::NewEnumerator2
    Sprite Size Mode=::NewEnumerator1
    Uniform Sprite Size Max=120
    Uniform Sprite Size Min=80
    Write Material Random=true
   )
   &SampleParticlesFromOtherEmitter(
    Maximum Sampled Particle Age=0.1
    Module.MaxSampledAgeCutoff=true
    Module.MinSampledAgeCutoff=false
    Reject Particle By Sampled Age=true
    Velocity Sampling=::NewEnumerator0
   )
   &ShapeLocation(MaskSpawnGroup=false, Offset Mode=::NewEnumerator0, Override Seed=false, Sphere Radius=25, UseOwnerScale=true)
   &AddVelocity(
    Module.Use Falloff Distance=false
    Override Seed=false
    Velocity Mode=::NewEnumerator1
    Velocity Origin=$Output.SampleParticlesFromOtherEmitter.SampledPosition
    Velocity Speed=&Multiply_Float(
     A=&RandomRangeFloat(Minimum=150, Maximum=250)
     B=&FN_FrontEndFloatModifier(Gameplay=1, Front End=1)
    )
   )
   &CameraOffset(Camera Offset Amount=&RandomRangeFloat(Minimum=-10, Maximum=-50, Evaluation Type=::NewEnumerator1))
   /SetVariables(Particles.SampledDynAttr:Vector4f=$Output.SampleParticlesFromOtherEmitter.DynamicMaterialParameter3) {enabled=false, isSetParameters=true, module=SetVariables_#5}
   &AddVelocity(Velocity=<0 0 0>, Velocity Speed Scale=2) {module=AddVelocity001}
   &self:GetColorPickAttr(
    Particle Attribute Reader=DataInterface() {propertyValues {EmitterBinding {bindingMode=Other, emitterName=Flare_With_Streak}}}
    Particle Index:NiagaraInt32=&MakeIntFromFloat(Float=$Particles.SampleParticlesFromOtherEmitter.SampledIndex)
   ) {module=GetColorPickAttr001}
  ]
  particleUpdateScript [
   &ParticleState()
   &ScaleSpriteSize(
    Scale Sprite Size Mode=::NewEnumerator0
    Uniform Scale Factor=&FloatFromCurve(
     FloatCurve=DataInterface() {propertyValues {Curve {keys [
         {interpMode=RCIM_Cubic, tangentMode=RCTM_User, time=0, value=0.3}
         {arriveTangent=0.5158, interpMode=RCIM_Cubic, leaveTangent=0.5158, tangentMode=RCTM_User, time=0.4, value=0.9}
         {arriveTangentWeight=0.3824, interpMode=RCIM_Cubic, tangentMode=RCTM_User, time=1, value=1}
        ]}}}
     Scale Curve=&FN_FrontEndFloatModifier(Gameplay=1.45, Front End=0.75)
    )
   )
   &Drag(
    Drag=&FN_FrontEndFloatModifier(Gameplay=16, Front End=16)
    Use Linear Drag=true
    Use Rotational Drag=true
   )
   &GravityForce(Gravity=<0 0 -222>) {enabled=false}
   &InheritSourceMovement(
    Applied Position Scale=&Multiply_VectorByFloat(Vector=<1 1 1>, Float=&Lerp_Float(A=0.5, B=0.8, Alpha=$System.Timeline.NormalizedTime))
    ApplyForce=false
    ApplyPosition=true
    ApplyPositionFalloff=false
   )
   &SolveForcesAndVelocity(Clamp Velocity=false, Limit Acceleration=false)
   &SpriteRotationRate(Rotation Rate=&RandomRangeFloat(Minimum=-0.03, Maximum=0.03))
   &ScaleColor(
    Scale Alpha=&FloatFromCurve(
     FloatCurve=DataInterface() {propertyValues {Curve {keys [
         {arriveTangent=9.0725, interpMode=RCIM_Cubic, leaveTangent=9.0725, tangentMode=RCTM_User, time=0, value=0}
         {arriveTangent=-2.5249, interpMode=RCIM_Cubic, leaveTangent=-2.5249, tangentMode=RCTM_User, time=0.1408, value=0.9847}
         {arriveTangent=-0.0405, interpMode=RCIM_Cubic, leaveTangent=-0.0405, tangentMode=RCTM_User, time=1, value=0}
        ]}}}
     Scale Curve=0.5
    )
    Scale RGB=&VectorFromFloat(Value=$System.TimeOfDayBrightnessMultiplier)
    ScaleA=true
    ScaleRGB=true
   )
   &DynamicMaterialParameters(
    Index 0 Param 1=&Add_Float(
     A=&RandomRangeFloat(Minimum=0, Maximum=1)
     B=&FloatFromCurve(
      FloatCurve=DataInterface() {propertyValues {Curve {keys [
          {arriveTangent=2.7533, interpMode=RCIM_Cubic, leaveTangent=1.8958, tangentMode=RCTM_User, time=0, value=1}
          {arriveTangent=0.8292, interpMode=RCIM_Cubic, leaveTangent=0.8292, tangentMode=RCTM_User, time=0.321, value=1.3453}
          {arriveTangent=0.0529, interpMode=RCIM_Cubic, leaveTangent=0.2021, tangentMode=RCTM_User, time=1, value=1.5}
         ]}}}
      CurveIndex=&Add_Float(A=&Multiply_Float(A=$Particles.Initial.MaterialRandom, B=0.5), B=$Particles.NormalizedAge)
      Scale Curve=&Lerp_Float(A=1, B=2, Alpha=0.5)
     )
    )
    Index 0 Param 2=&FloatFromCurve(FloatCurve=DataInterface() {propertyValues {Curve {keys [<at 0 1>,<at 1 0>]}}})
    Index 0 Param 3=&FloatFromCurve(
     FloatCurve=DataInterface() {propertyValues {Curve {keys [
         {arriveTangent=-2.2052, interpMode=RCIM_Cubic, leaveTangent=-2.2052, tangentMode=RCTM_User, time=0.1, value=1}
         {interpMode=RCIM_Cubic, tangentMode=RCTM_User, time=1, value=0.4}
        ]}}}
     Scale Curve=&Multiply_Float(A=0.5, B=$System.TimeOfDayBrightnessMultiplier)
    )
    Index 0 Param 4=&FloatFromCurve(FloatCurve=DataInterface() {propertyValues {Curve {keys [{interpMode=RCIM_Cubic, tangentMode=RCTM_Break, time=0.5, value=1}, {interpMode=RCIM_Cubic, tangentMode=RCTM_Break, time=1, value=0}]}}})
    Index 1 Param 1=&FloatFromCurve(
     FloatCurve=DataInterface() {propertyValues {Curve {keys [
         {interpMode=RCIM_Cubic, leaveTangentWeight=0.0833, tangentMode=RCTM_Break, time=0, value=0}
         {arriveTangentWeight=0.0873, interpMode=RCIM_Cubic, tangentMode=RCTM_Break, time=0.25, value=1}
        ]}}}
     Scale Curve=&Lerp_Float(A=0.35, B=0.6, Alpha=0.5)
    )
    Index 1 Param 2=&FloatFromCurve(
     FloatCurve=DataInterface() {propertyValues {Curve {keys [
         {interpMode=RCIM_Cubic, tangentMode=RCTM_Break, time=0, value=0.2}
         {interpMode=RCIM_Cubic, leaveTangent=0.9066, tangentMode=RCTM_Break, time=1, value=2}
        ]}}}
     Scale Curve=&RandomRangeFloat(Minimum=0.1, Maximum=0.2)
    )
    Index 1 Param 3=&FloatFromCurve(FloatCurve=DataInterface() {propertyValues {Curve {keys [
         {interpMode=RCIM_Cubic, leaveTangent=6.3917, tangentMode=RCTM_Break, time=0, value=0}
         {arriveTangent=1.7549, interpMode=RCIM_Cubic, leaveTangent=10.6767, tangentMode=RCTM_User, time=0.25, value=1}
        ]}}})
    Index 1 Param 4=&FloatFromCurve(
     FloatCurve=DataInterface() {propertyValues {Curve {keys [{arriveTangent=-0.136, interpMode=RCIM_Cubic, leaveTangent=-0.136, tangentMode=RCTM_User, time=0, value=1}, <at 1 0.95>]}}}
     Scale Curve=&Lerp_Float(A=2, B=5, Alpha=0.5)
    )
    Index 2 Param 1=&FloatFromCurve(FloatCurve=DataInterface() {propertyValues {Curve {keys [
         {interpMode=RCIM_Cubic, leaveTangent=4.2485, tangentMode=RCTM_Break, time=0, value=0}
         {arriveTangent=1.0548, interpMode=RCIM_Cubic, leaveTangent=1.0548, tangentMode=RCTM_User, time=0.17, value=0.5715}
         <at 1 1>
        ]}}})
    Index 2 Param 2=&FloatFromCurve(
     FloatCurve=DataInterface() {propertyValues {Curve {keys [
         {arriveTangent=1.9056, interpMode=RCIM_Cubic, leaveTangent=1.9056, tangentMode=RCTM_User, time=0, value=0.2}
         {arriveTangent=0.0259, interpMode=RCIM_Cubic, leaveTangent=0.0259, tangentMode=RCTM_User, time=1, value=1.4}
        ]}}}
     Scale Curve=1.5
    )
    Index 2 Param 3=&FloatFromCurve(
     FloatCurve=DataInterface() {propertyValues {Curve {keys [
         {interpMode=RCIM_Cubic, leaveTangentWeight=0.1411, tangentMode=RCTM_Break, time=0, value=2}
         {interpMode=RCIM_Cubic, tangentMode=RCTM_Break, time=0.3, value=1}
         {interpMode=RCIM_Cubic, leaveTangent=11.1233, tangentMode=RCTM_Break, time=1, value=2}
        ]}}}
     Scale Curve=10
    )
    Index 2 Param 4=&FloatFromCurve(
     FloatCurve=DataInterface() {propertyValues {Curve {keys [
         {arriveTangent=-0.6867, interpMode=RCIM_Cubic, leaveTangent=-0.6867, tangentMode=RCTM_User, time=0, value=1}
         {interpMode=RCIM_Cubic, leaveTangent=-1.2956, tangentMode=RCTM_Break, time=1, value=0.75}
        ]}}}
     Scale Curve=&RandomRangeFloat(Minimum=0.5, Maximum=2)
    )
    Module.Param0WriteEnabled=true
    Module.Param0WriteEnabled1=true
    Module.Param0WriteEnabled2=true
    Module.Param1WriteEnabled=false
    Module.Param1WriteEnabled1=true
    Module.Param1WriteEnabled2=true
    Module.Param2WriteEnabled=true
    Module.Param2WriteEnabled1=true
    Module.Param2WriteEnabled2=true
    Module.Param3WriteEnabled=true
    Module.Param3WriteEnabled1=true
    Module.Param3WriteEnabled2=true
    Write Parameter Index 1=true
    Write Parameter Index 2=true
   )
   &self:ScratchModule_01(Particle Attribute Reader=DataInterface() {propertyValues {EmitterBinding {bindingMode=Other, emitterName=Flare_With_Streak}}}) {enabled=false, module=ScratchModule_01}
   &self:GetColorPickAttr(
    Particle Attribute Reader=DataInterface() {propertyValues {EmitterBinding {bindingMode=Other, emitterName=Flare_With_Streak}}}
    Particle Index:NiagaraInt32=&MakeIntFromFloat(Float=$Particles.SampleParticlesFromOtherEmitter.SampledIndex)
   ) {enabled=false, module=GetColorPickAttr}
   /SetVariables(Particles.PositionOffset:NiagaraPosition=&AddVectorToPosition(Position=$Particles.Position, Vector=&MakeVector(X=0, Y=0, Z=$Particles.Offset))) {enabled=false, isSetParameters=true, module=SetVariables_#6}
   /SetVariables(Particles.VisibilityTag=&SelectIntFromArray(Direct Array Index=$Particles.ColorPick, Int Selection Array=DataInterface() {propertyValues {IntData [0,1,2]}})) {isSetParameters=true, module=SetVariables_#7}
   /SetVariables(Particles.Color=&SelectLinearColorFromArray(
     Array Sampling Mode=::NewEnumerator1
     Direct Array Index=$Particles.ColorPick
     Color Selection Array=DataInterface() {propertyValues {ColorData [<rgba 1.5 1 0 0.075>, <rgba 1 1 1 0.025>, <rgba 6 0.6 0 0.075>]}}
     Random Seed=?
    )) {isSetParameters=true, module=SetVariables_#8}
  ]
  rendererClasses [&NiagaraSpriteRendererProperties]
  renderers [
   {rendererClass=&NiagaraSpriteRendererProperties}
   {rendererClass=&NiagaraSpriteRendererProperties, rendererIndex=1}
   {rendererClass=&NiagaraSpriteRendererProperties, rendererIndex=2}
   {rendererClass=&NiagaraSpriteRendererProperties, rendererIndex=3}
   {rendererClass=&NiagaraSpriteRendererProperties, rendererIndex=4}
   {rendererClass=&NiagaraSpriteRendererProperties, rendererIndex=5}
  ]
 }
 Dust_FE {
  emitterUpdateScript [
   &EmitterState()
   &SpawnBurst_Instantaneous(Spawn Count=8, Use Spawn Probability=false, UseLoopCountLimit=false) {enabled=false}
   &SpawnParticlesFromOtherEmitter(
    Attribute Reader=DataInterface() {propertyValues {EmitterBinding {bindingMode=Other, emitterName=Flare_With_Streak_FE}}}
    Maximum Age Spawn Threshold=0
    Spawn Rate=&FN_FrontEndFloatModifier(Gameplay=0, Front End=250)
    Use Maximum Age=false
    Use Minimum Age=false
   )
  ]
  particleSpawnScript [
   &InitializeParticle(
    Color Mode=::NewEnumerator0
    Lifetime Max=1
    Lifetime Min=0.5
    Lifetime Mode=::NewEnumerator1
    Position Mode=::NewEnumerator0
    Sprite Rotation Mode=::NewEnumerator2
    Sprite Size Mode=::NewEnumerator1
    Uniform Sprite Size Max=120
    Uniform Sprite Size Min=80
    Write Material Random=true
   )
   &SampleParticlesFromOtherEmitter(Maximum Sampled Particle Age=0.1, Module.MaxSampledAgeCutoff=true, Module.MinSampledAgeCutoff=false, Reject Particle By Sampled Age=true)
   &ShapeLocation(MaskSpawnGroup=false, Offset Mode=::NewEnumerator0, Override Seed=false, Sphere Radius=10, UseOwnerScale=true)
   &AddVelocity(
    Module.Use Falloff Distance=false
    Override Seed=false
    Velocity Mode=::NewEnumerator1
    Velocity Origin=$Output.SampleParticlesFromOtherEmitter.SampledPosition
    Velocity Speed=&Multiply_Float(
     A=&RandomRangeFloat(Minimum=150, Maximum=250)
     B=&FN_FrontEndFloatModifier(Gameplay=1, Front End=0.5)
    )
   )
   &CameraOffset(Camera Offset Amount=&RandomRangeFloat(Minimum=-10, Maximum=-50, Evaluation Type=::NewEnumerator1))
   /SetVariables(Particles.SampledDynAttr:Vector4f=$Output.SampleParticlesFromOtherEmitter.DynamicMaterialParameter3) {enabled=false, isSetParameters=true, module=SetVariables_#5}
   &AddVelocity(Velocity=<0 0 -1>, Velocity Speed Scale=200) {module=AddVelocity001}
   &self:GetColorPickAttr(
    Particle Attribute Reader=DataInterface() {propertyValues {EmitterBinding {bindingMode=Other, emitterName=Flare_With_Streak_FE}}}
    Particle Index:NiagaraInt32=&MakeIntFromFloat(Float=$Particles.SampleParticlesFromOtherEmitter.SampledIndex)
   ) {module=GetColorPickAttr001}
  ]
  particleUpdateScript [
   &ParticleState()
   &ScaleSpriteSize(
    Scale Sprite Size Mode=::NewEnumerator0
    Uniform Scale Factor=&FloatFromCurve(
     FloatCurve=DataInterface() {propertyValues {Curve {keys [
         {interpMode=RCIM_Cubic, tangentMode=RCTM_User, time=0, value=0.3}
         {arriveTangent=0.5158, interpMode=RCIM_Cubic, leaveTangent=0.5158, tangentMode=RCTM_User, time=0.4, value=0.9}
         {arriveTangentWeight=0.3824, interpMode=RCIM_Cubic, tangentMode=RCTM_User, time=1, value=1}
        ]}}}
     Scale Curve=&FN_FrontEndFloatModifier(Gameplay=1.45, Front End=&Multiply_Float(A=0.75, B=$System.FE_SizeModifier))
    )
   )
   &Drag(
    Drag=&FN_FrontEndFloatModifier(Gameplay=16, Front End=16)
    Use Linear Drag=true
    Use Rotational Drag=true
   )
   &GravityForce(Gravity=<0 0 -222>) {enabled=false}
   &InheritSourceMovement(
    Applied Position Scale=&Multiply_VectorByFloat(Vector=<1 1 1>, Float=&Lerp_Float(A=0.5, B=0.8, Alpha=$System.Timeline.NormalizedTime))
    ApplyForce=false
    ApplyPosition=true
    ApplyPositionFalloff=false
   )
   &SolveForcesAndVelocity(Clamp Velocity=false, Limit Acceleration=false)
   &SpriteRotationRate(Rotation Rate=&RandomRangeFloat(Minimum=-0.03, Maximum=0.03))
   &ScaleColor(
    Scale Alpha=&FloatFromCurve(
     FloatCurve=DataInterface() {propertyValues {Curve {keys [
         {arriveTangent=9.0725, interpMode=RCIM_Cubic, leaveTangent=9.0725, tangentMode=RCTM_User, time=0, value=0}
         {arriveTangent=-2.5249, interpMode=RCIM_Cubic, leaveTangent=-2.5249, tangentMode=RCTM_User, time=0.1408, value=0.9847}
         {arriveTangent=-0.0405, interpMode=RCIM_Cubic, leaveTangent=-0.0405, tangentMode=RCTM_User, time=1, value=0}
        ]}}}
     Scale Curve=0.5
    )
    Scale RGB=&VectorFromFloat(Value=$System.TimeOfDayBrightnessMultiplier)
    ScaleA=true
    ScaleRGB=true
   )
   &DynamicMaterialParameters(
    Index 0 Param 1=&Add_Float(
     A=&RandomRangeFloat(Minimum=0, Maximum=1)
     B=&FloatFromCurve(
      FloatCurve=DataInterface() {propertyValues {Curve {keys [
          {arriveTangent=2.7533, interpMode=RCIM_Cubic, leaveTangent=1.8958, tangentMode=RCTM_User, time=0, value=1}
          {arriveTangent=0.8292, interpMode=RCIM_Cubic, leaveTangent=0.8292, tangentMode=RCTM_User, time=0.321, value=1.3453}
          {arriveTangent=0.0529, interpMode=RCIM_Cubic, leaveTangent=0.2021, tangentMode=RCTM_User, time=1, value=1.5}
         ]}}}
      CurveIndex=&Add_Float(A=&Multiply_Float(A=$Particles.Initial.MaterialRandom, B=0.5), B=$Particles.NormalizedAge)
      Scale Curve=&Lerp_Float(A=1, B=2, Alpha=0.5)
     )
    )
    Index 0 Param 2=&FloatFromCurve(FloatCurve=DataInterface() {propertyValues {Curve {keys [<at 0 1>,<at 1 0>]}}})
    Index 0 Param 3=&FloatFromCurve(
     FloatCurve=DataInterface() {propertyValues {Curve {keys [
         {arriveTangent=-2.2052, interpMode=RCIM_Cubic, leaveTangent=-2.2052, tangentMode=RCTM_User, time=0.1, value=1}
         {interpMode=RCIM_Cubic, tangentMode=RCTM_User, time=1, value=0.4}
        ]}}}
     Scale Curve=&Multiply_Float(A=0.5, B=$System.TimeOfDayBrightnessMultiplier)
    )
    Index 0 Param 4=&FloatFromCurve(FloatCurve=DataInterface() {propertyValues {Curve {keys [{interpMode=RCIM_Cubic, tangentMode=RCTM_Break, time=0.5, value=1}, {interpMode=RCIM_Cubic, tangentMode=RCTM_Break, time=1, value=0}]}}})
    Index 1 Param 1=&FloatFromCurve(
     FloatCurve=DataInterface() {propertyValues {Curve {keys [
         {interpMode=RCIM_Cubic, leaveTangentWeight=0.0833, tangentMode=RCTM_Break, time=0, value=0}
         {arriveTangentWeight=0.0873, interpMode=RCIM_Cubic, tangentMode=RCTM_Break, time=0.25, value=1}
        ]}}}
     Scale Curve=&Lerp_Float(A=0.35, B=0.6, Alpha=0.5)
    )
    Index 1 Param 2=&FloatFromCurve(
     FloatCurve=DataInterface() {propertyValues {Curve {keys [
         {interpMode=RCIM_Cubic, tangentMode=RCTM_Break, time=0, value=0.2}
         {interpMode=RCIM_Cubic, leaveTangent=0.9066, tangentMode=RCTM_Break, time=1, value=2}
        ]}}}
     Scale Curve=&RandomRangeFloat(Minimum=0.1, Maximum=0.2)
    )
    Index 1 Param 3=&FloatFromCurve(FloatCurve=DataInterface() {propertyValues {Curve {keys [
         {interpMode=RCIM_Cubic, leaveTangent=6.3917, tangentMode=RCTM_Break, time=0, value=0}
         {arriveTangent=1.7549, interpMode=RCIM_Cubic, leaveTangent=10.6767, tangentMode=RCTM_User, time=0.25, value=1}
        ]}}})
    Index 1 Param 4=&FloatFromCurve(
     FloatCurve=DataInterface() {propertyValues {Curve {keys [{arriveTangent=-0.136, interpMode=RCIM_Cubic, leaveTangent=-0.136, tangentMode=RCTM_User, time=0, value=1}, <at 1 0.95>]}}}
     Scale Curve=&Lerp_Float(A=2, B=5, Alpha=0.5)
    )
    Index 2 Param 1=&FloatFromCurve(FloatCurve=DataInterface() {propertyValues {Curve {keys [
         {interpMode=RCIM_Cubic, leaveTangent=4.2485, tangentMode=RCTM_Break, time=0, value=0}
         {arriveTangent=1.0548, interpMode=RCIM_Cubic, leaveTangent=1.0548, tangentMode=RCTM_User, time=0.17, value=0.5715}
         <at 1 1>
        ]}}})
    Index 2 Param 2=&FloatFromCurve(
     FloatCurve=DataInterface() {propertyValues {Curve {keys [
         {arriveTangent=1.9056, interpMode=RCIM_Cubic, leaveTangent=1.9056, tangentMode=RCTM_User, time=0, value=0.2}
         {arriveTangent=0.0259, interpMode=RCIM_Cubic, leaveTangent=0.0259, tangentMode=RCTM_User, time=1, value=1.4}
        ]}}}
     Scale Curve=1.5
    )
    Index 2 Param 3=&FloatFromCurve(
     FloatCurve=DataInterface() {propertyValues {Curve {keys [
         {interpMode=RCIM_Cubic, leaveTangentWeight=0.1411, tangentMode=RCTM_Break, time=0, value=2}
         {interpMode=RCIM_Cubic, tangentMode=RCTM_Break, time=0.3, value=1}
         {interpMode=RCIM_Cubic, leaveTangent=11.1233, tangentMode=RCTM_Break, time=1, value=2}
        ]}}}
     Scale Curve=10
    )
    Index 2 Param 4=&FloatFromCurve(
     FloatCurve=DataInterface() {propertyValues {Curve {keys [
         {arriveTangent=-0.6867, interpMode=RCIM_Cubic, leaveTangent=-0.6867, tangentMode=RCTM_User, time=0, value=1}
         {interpMode=RCIM_Cubic, leaveTangent=-1.2956, tangentMode=RCTM_Break, time=1, value=0.75}
        ]}}}
     Scale Curve=&RandomRangeFloat(Minimum=0.5, Maximum=2)
    )
    Module.Param0WriteEnabled=true
    Module.Param0WriteEnabled1=true
    Module.Param0WriteEnabled2=true
    Module.Param1WriteEnabled=false
    Module.Param1WriteEnabled1=true
    Module.Param1WriteEnabled2=true
    Module.Param2WriteEnabled=true
    Module.Param2WriteEnabled1=true
    Module.Param2WriteEnabled2=true
    Module.Param3WriteEnabled=true
    Module.Param3WriteEnabled1=true
    Module.Param3WriteEnabled2=true
    Write Parameter Index 1=true
    Write Parameter Index 2=true
   )
   &self:ScratchModule_01(Particle Attribute Reader=DataInterface() {propertyValues {EmitterBinding {bindingMode=Other, emitterName=Flare_With_Streak}}}) {enabled=false, module=ScratchModule_01}
   &self:GetColorPickAttr(
    Particle Attribute Reader=DataInterface() {propertyValues {EmitterBinding {bindingMode=Other, emitterName=Flare_With_Streak}}}
    Particle Index:NiagaraInt32=&MakeIntFromFloat(Float=$Particles.SampleParticlesFromOtherEmitter.SampledIndex)
   ) {enabled=false, module=GetColorPickAttr}
   /SetVariables(Particles.PositionOffset:NiagaraPosition=&AddVectorToPosition(Position=$Particles.Position, Vector=&MakeVector(X=0, Y=0, Z=$Particles.Offset))) {enabled=false, isSetParameters=true, module=SetVariables_#6}
   /SetVariables(Particles.VisibilityTag=&SelectIntFromArray(Direct Array Index=$Particles.ColorPick, Int Selection Array=DataInterface() {propertyValues {IntData [0,1,2]}})) {isSetParameters=true, module=SetVariables_#7}
   /SetVariables(Particles.Color=&SelectLinearColorFromArray(
     Array Sampling Mode=::NewEnumerator1
     Direct Array Index=$Particles.ColorPick
     Color Selection Array=DataInterface() {propertyValues {ColorData [<rgba 1.5 1 0 0.075>, <rgba 1 1 1 0.025>, <rgba 6 0.6 0 0.075>]}}
     Random Seed=?
    )) {isSetParameters=true, module=SetVariables_#8}
   &self:GetOffsetAttr(
    Particle Attribute Reader=DataInterface() {propertyValues {EmitterBinding {bindingMode=Other, emitterName=Flare_With_Streak_FE}}}
    Particle Index:NiagaraInt32=&MakeIntFromFloat(Float=$Particles.SampleParticlesFromOtherEmitter.SampledIndex)
   ) {module=GetOffsetAttr}
   /SetVariables(Particles.FEPosition:NiagaraPosition=&Lerp_Vector(
     A=Linked() {linkedVariable {name=Particles.Position, type=NiagaraPosition}}
     B=&AddVectorToPosition(Position=$Particles.Position, Vector=$Particles.Offset)
     Alpha=&FloatFromCurve(FloatCurve=DataInterface() {propertyValues {Curve {defaultValue=1, keys [{tangentMode=RCTM_User, time=0, value=0}, {tangentMode=RCTM_User, time=1, value=1}]}}})
    )) {enabled=false, isSetParameters=true}
   /SetVariables(Particles.Offset:Vector3f=$System.Offset) {isSetParameters=true}
   /SetVariables(Particles.Position=&AddVectorToPosition(Position=$Particles.Position, Vector=$Particles.Offset)) {isSetParameters=true}
   &CameraOffset(Camera Offset Amount=$System.FE_CameraOffset) {module=CameraOffset001}
  ]
  rendererClasses [&NiagaraSpriteRendererProperties]
  renderers [
   {rendererClass=&NiagaraSpriteRendererProperties}
   {rendererClass=&NiagaraSpriteRendererProperties, rendererIndex=1}
   {rendererClass=&NiagaraSpriteRendererProperties, rendererIndex=2}
   {rendererClass=&NiagaraSpriteRendererProperties, rendererIndex=3}
   {rendererClass=&NiagaraSpriteRendererProperties, rendererIndex=4}
   {rendererClass=&NiagaraSpriteRendererProperties, rendererIndex=5}
  ]
 }
 FireworkFlare_Parent_Sprite {
  emitterUpdateScript [
   &EmitterState()
   &SpawnBurst_Instantaneous(Spawn Count=12, Use Spawn Probability=false, UseLoopCountLimit=false) {enabled=false}
   &SpawnParticlesFromOtherEmitter(
    Attribute Reader=DataInterface() {propertyValues {EmitterBinding {bindingMode=Other, emitterName=Flare_With_Streak}}}
    Maximum Age Spawn Threshold=0
    Spawn Rate=&FN_FrontEndFloatModifier(Gameplay=40, Front End=0)
    Spawn Rate Per Particle Cap=60
    Use Maximum Age=false
    Use Minimum Age=false
   )
  ]
  particleSpawnScript [
   &InitializeParticle(
    AdjustAlpha=true
    AdjustHue=true
    AdjustSaturation=true
    AdjustValue=true
    Color=<rgba 0 0.0833 1 0.666>
    Color Mode=::NewEnumerator3
    Lifetime Mode=::NewEnumerator1
    Sprite Size Mode=::NewEnumerator1
    Uniform Sprite Size Max=150
    Uniform Sprite Size Min=100
    UsePositionOffset=false
    Write Material Random=true
   )
   &ShapeLocation(
    MaskSpawnGroup=false
    Override Seed=false
    Rotation Mode=::NewEnumerator5
    Scale Mode:ENiagara_ScaleMode=::NewEnumerator1
    Shape Origin=&TransformPosition(Position=Linked() {linkedVariable {name=System.ArrowTipPosition, type=Vector3f}}, Source Space=::Local, Destination Space=::Simulation)
    Sphere Radius=10
   )
   &AddVelocity(
    Velocity=&Multiply_Vector(A=$Particles.ShapeLocation.ShapeVector, B=<0.75 0.75 0.333>)
    Velocity Speed Scale=&Multiply_Float(
     A=&RandomRangeFloat(Minimum=60, Maximum=120)
     B=&FN_FrontEndFloatModifier(Gameplay=1, Front End=1)
    )
   )
   &SampleParticlesFromOtherEmitter(
    Maximum Sampled Particle Age=0.1
    Module.MaxSampledAgeCutoff=true
    Module.MinSampledAgeCutoff=false
    Reject Particle By Sampled Age=true
    Velocity Sampling=::NewEnumerator0
   )
   &self:GetColorPickAttr(
    Particle Attribute Reader=DataInterface() {propertyValues {EmitterBinding {bindingMode=Other, emitterName=Flare_With_Streak}}}
    Particle Index:NiagaraInt32=&MakeIntFromFloat(Float=$Particles.SampleParticlesFromOtherEmitter.SampledIndex)
   ) {module=GetColorPickAttr}
  ]
  particleUpdateScript [
   &Drag(Drag=&RandomRangeFloat(Minimum=2, Maximum=4), Use Linear Drag=true, Use Rotational Drag=false)
   &InheritSourceMovement(
    Applied Position Scale=&Multiply_VectorByFloat(Vector=<1 1 1>, Float=&Lerp_Float(A=0.5, B=0.8, Alpha=$System.Timeline.NormalizedTime))
    ApplyForce=false
    ApplyPosition=true
    ApplyPositionFalloff=false
   )
   &GravityForce(Gravity=&RandomRangeVector(Minimum=<0 0 -150>, Maximum=<0 0 -200>))
   &CurlNoiseForce(Noise Frequency=10, Noise Strength=100, Pan Noise=false) {module=CurlNoiseForce001}
   &ParticleState()
   &ScaleSpriteSize(
    Uniform Curve Scale=&Multiply_Float(
     A=&RandomRangeFloat(Minimum=0.75, Maximum=1.25)
     B=&FN_FrontEndFloatModifier(Gameplay=1, Front End=0.5)
    )
    Uniform Curve Sprite Scale=DataInterface() {propertyValues {Curve {keys [
        {arriveTangent=2.9868, time=0, value=0}
        {arriveTangent=-1.2213, interpMode=RCIM_Cubic, leaveTangent=-1.2213, tangentMode=RCTM_User, time=0.1, value=0.833}
        {arriveTangent=-0.8542, leaveTangent=-0.0405, tangentMode=RCTM_User, time=0.7, value=0}
       ]}}}
   )
   &DynamicMaterialParameters(
    Index 0 Param 1=0.8
    Index 0 Param 2=1
    Index 0 Param 3=1
    Index 1 Param 1=150
    Index 1 Param 4=1
    Module.Param0WriteEnabled=true
    Module.Param0WriteEnabled1=true
    Module.Param1WriteEnabled=true
    Module.Param1WriteEnabled1=true
    Module.Param2WriteEnabled=true
    Module.Param2WriteEnabled1=true
    Module.Param3WriteEnabled=false
    Module.Param3WriteEnabled1=true
    Write Parameter Index 1=true
   )
   &ScaleColor(
    Scale RGB=&VectorFromFloat(Value=&Multiply_Float(
      A=$System.TimeOfDayBrightnessMultiplier
      B=&FloatFromCurve(FloatCurve=DataInterface() {propertyValues {Curve {keys [{arriveTangent=-523.03, interpMode=RCIM_Cubic, leaveTangent=-523.03, tangentMode=RCTM_User, time=0, value=30}, <at 0.1 1>]}}})
     ))
    ScaleA=true
    ScaleRGB=true
   ) {enabled=false, module=ScaleColor002}
   &SolveForcesAndVelocity(Clamp Velocity=false, Limit Acceleration=false)
   /SetVariables(Particles.Color=&SelectLinearColorFromArray(
     Array Sampling Mode=::NewEnumerator1
     Direct Array Index=$Particles.ColorPick
     Color Selection Array=DataInterface() {propertyValues {ColorData [<rgba 6 4 0 0.1>, <rgba 0.5 0.5 0.5 0.5>, <rgba 6 0.6 0 0.1>]}}
     Random Seed=?
    )) {isSetParameters=true, module=SetVariables_#9}
  ]
  rendererClasses [&NiagaraSpriteRendererProperties]
  renderers [{rendererClass=&NiagaraSpriteRendererProperties}]
 }
 FireworkFlare_Parent_Sprite_FE {
  emitterUpdateScript [
   &EmitterState()
   &SpawnBurst_Instantaneous(Spawn Count=12, Use Spawn Probability=false, UseLoopCountLimit=false) {enabled=false}
   &SpawnParticlesFromOtherEmitter(
    Attribute Reader=DataInterface() {propertyValues {EmitterBinding {bindingMode=Other, emitterName=Flare_With_Streak_FE}}}
    Maximum Age Spawn Threshold=0
    Spawn Rate=40
    Spawn Rate Per Particle Cap=40
    Use Maximum Age=false
    Use Minimum Age=false
   )
  ]
  particleSpawnScript [
   &InitializeParticle(
    AdjustAlpha=true
    AdjustHue=true
    AdjustSaturation=true
    AdjustValue=true
    Color=<rgba 0 0.0833 1 0.666>
    Color Mode=::NewEnumerator3
    Lifetime Max=1
    Lifetime Min=0.5
    Lifetime Mode=::NewEnumerator1
    Sprite Size Mode=::NewEnumerator1
    Uniform Sprite Size Max=150
    Uniform Sprite Size Min=100
    UsePositionOffset=false
    Write Material Random=true
   )
   &ShapeLocation(
    MaskSpawnGroup=false
    Override Seed=false
    Rotation Mode=::NewEnumerator5
    Scale Mode:ENiagara_ScaleMode=::NewEnumerator1
    Shape Origin=&TransformPosition(Position=Linked() {linkedVariable {name=System.ArrowTipPosition, type=Vector3f}}, Source Space=::Local, Destination Space=::Simulation)
    Sphere Radius=10
   )
   &AddVelocity(
    Velocity=&Multiply_Vector(A=$Particles.ShapeLocation.ShapeVector, B=<0.75 0.75 0.333>)
    Velocity Speed Scale=&Multiply_Float(
     A=&RandomRangeFloat(Minimum=60, Maximum=120)
     B=&FN_FrontEndFloatModifier(Gameplay=1, Front End=&Multiply_Float(A=1, B=$System.FE_SizeModifier))
    )
   )
   &SampleParticlesFromOtherEmitter(
    Maximum Sampled Particle Age=0.1
    Module.MaxSampledAgeCutoff=true
    Module.MinSampledAgeCutoff=false
    Reject Particle By Sampled Age=true
    Velocity Sampling=::NewEnumerator0
   )
   &self:GetColorPickAttr(
    Particle Attribute Reader=DataInterface() {propertyValues {EmitterBinding {bindingMode=Other, emitterName=Flare_With_Streak_FE}}}
    Particle Index:NiagaraInt32=&MakeIntFromFloat(Float=$Particles.SampleParticlesFromOtherEmitter.SampledIndex)
   ) {module=GetColorPickAttr}
  ]
  particleUpdateScript [
   &Drag(Drag=&RandomRangeFloat(Minimum=2, Maximum=4), Use Linear Drag=true, Use Rotational Drag=false)
   &InheritSourceMovement(
    Applied Position Scale=&Multiply_VectorByFloat(Vector=<1 1 1>, Float=&Lerp_Float(A=0.5, B=0.8, Alpha=$System.Timeline.NormalizedTime))
    ApplyForce=false
    ApplyPosition=true
    ApplyPositionFalloff=false
   )
   &GravityForce(Gravity=&RandomRangeVector(Minimum=<0 0 -150>, Maximum=<0 0 -200>))
   &CurlNoiseForce(Noise Frequency=10, Noise Strength=100, Pan Noise=false) {module=CurlNoiseForce001}
   &ParticleState()
   &ScaleSpriteSize(
    Uniform Curve Scale=&Multiply_Float(
     A=&RandomRangeFloat(Minimum=0.75, Maximum=1.25)
     B=&FN_FrontEndFloatModifier(Gameplay=1, Front End=0.5)
    )
    Uniform Curve Sprite Scale=DataInterface() {propertyValues {Curve {keys [
        {arriveTangent=2.9868, time=0, value=0}
        {arriveTangent=-1.2213, interpMode=RCIM_Cubic, leaveTangent=-1.2213, tangentMode=RCTM_User, time=0.1, value=0.833}
        {arriveTangent=-0.8542, leaveTangent=-0.0405, tangentMode=RCTM_User, time=0.7, value=0}
       ]}}}
   )
   &DynamicMaterialParameters(
    Index 0 Param 1=0.8
    Index 0 Param 2=1
    Index 0 Param 3=1
    Index 1 Param 1=150
    Index 1 Param 4=1
    Module.Param0WriteEnabled=true
    Module.Param0WriteEnabled1=true
    Module.Param1WriteEnabled=true
    Module.Param1WriteEnabled1=true
    Module.Param2WriteEnabled=true
    Module.Param2WriteEnabled1=true
    Module.Param3WriteEnabled=false
    Module.Param3WriteEnabled1=true
    Write Parameter Index 1=true
   )
   &ScaleColor(
    Scale RGB=&VectorFromFloat(Value=&Multiply_Float(
      A=$System.TimeOfDayBrightnessMultiplier
      B=&FloatFromCurve(FloatCurve=DataInterface() {propertyValues {Curve {keys [{arriveTangent=-523.03, interpMode=RCIM_Cubic, leaveTangent=-523.03, tangentMode=RCTM_User, time=0, value=30}, <at 0.1 1>]}}})
     ))
    ScaleA=true
    ScaleRGB=true
   ) {enabled=false, module=ScaleColor002}
   &SolveForcesAndVelocity(Clamp Velocity=false, Limit Acceleration=false)
   /SetVariables(Particles.Color=&SelectLinearColorFromArray(
     Array Sampling Mode=::NewEnumerator1
     Direct Array Index=$Particles.ColorPick
     Color Selection Array=DataInterface() {propertyValues {ColorData [<rgba 6 4 0 0.1>, <rgba 0.5 0.5 0.5 0.5>, <rgba 6 0.6 0 0.1>]}}
     Random Seed=?
    )) {isSetParameters=true, module=SetVariables_#9}
   /SetVariables(Particles.Offset:Vector3f=$System.Offset) {isSetParameters=true}
   /SetVariables(Particles.Position=&AddVectorToPosition(Position=$Particles.Position, Vector=$Particles.Offset)) {isSetParameters=true}
   &CameraOffset(Camera Offset Amount=$System.FE_CameraOffset)
  ]
  rendererClasses [&NiagaraSpriteRendererProperties]
  renderers [{rendererClass=&NiagaraSpriteRendererProperties}]
 }
 Flare_With_Streak {
  emitterUpdateScript [
   &EmitterState()
   &SpawnRate(
    Spawn Probability=0.3
    SpawnRate=&FN_FrontEndFloatModifier(Gameplay=&Lerp_Float(A=5, B=15, Alpha=$System.Timeline.NormalizedTime), Front End=0)
    Use Spawn Probability=true
   )
   &SpawnBurst_Instantaneous(Use Spawn Probability=false, UseLoopCountLimit=false) {enabled=false}
  ]
  particleSpawnScript [
   /SetVariables(Particles.ColorPick:NiagaraInt32=&RandomRangeInt(Minimum=0, Maximum=2)) {isSetParameters=true, module=SetVariables_#10}
   &InitializeParticle(
    Color=&SelectLinearColorFromArray(
     Array Sampling Mode=::NewEnumerator1
     Direct Array Index=$Particles.ColorPick
     Color Selection Array=DataInterface() {propertyValues {ColorData [<rgba 3 2 0 1>, <rgba 0.5 0.5 0.5 1>, <rgba 3 0.3 0 1>]}}
     Random Seed=?
    )
    Lifetime=0.5
    Position Offset=$System.ArrowTipPosition
    Position Offset Coordinate Space=::Local
    Sprite Rotation Mode=::NewEnumerator2
    Sprite Size Mode=::NewEnumerator3
    Uniform Sprite Size=&FN_FrontEndFloatModifier(Gameplay=500, Front End=250)
    UsePositionOffset=true
    Write Material Random=true
   )
   &DynamicMaterialParameters(
    Index 0 Param 1=&RandomRangeFloat(Minimum=0.4, Maximum=0.6)
    Index 0 Param 2=1
    Index 0 Param 3=1
    Index 0 Param 4=1
    Module.Index0Float0Write=true
    Module.Index0Float1Write=true
    Module.Index0Float2Write=true
    Module.Index0Float3Write=true
   )
   /SetVariables() {isSetParameters=true, module=SetVariables_#11}
   &ShapeLocation(
    Cylinder Height=&FN_FrontEndFloatModifier(Gameplay=300, Front End=0)
    Cylinder Radius=&FN_FrontEndFloatModifier(Gameplay=500, Front End=0)
    MaskSpawnGroup=false
    Offset=&FN_FrontEndVector3Modifier(Gameplay=&Lerp_Vector(A=<0 0 -600>, B=<0 0 -700>, Alpha=$System.Timeline.NormalizedTime), Front End=<0 0 0>)
    Offset Mode=::NewEnumerator0
    Override Seed=false
    Shape Primitive=::NewEnumerator1
    Surface Only=true
    Surface Only Band Thickness=0.35
    Use Endcaps In Surface Only Mode=false
    UseOwnerScale=true
   )
   &AddVelocity(
    Velocity=<0 0 1>
    Velocity Speed Scale=&FN_FrontEndFloatModifier(
     Gameplay=0
     Front End=&Multiply_Float(A=&Lerp_Float(A=0.5, B=1, Alpha=$System.Timeline.NormalizedTime), B=500)
    )
   ) {enabled=false}
  ]
  particleUpdateScript [
   &ParticleState()
   &InheritSourceMovement(
    Applied Position Scale=&Multiply_VectorByFloat(Vector=<1 1 1>, Float=&Lerp_Float(A=0.5, B=0.8, Alpha=$System.Timeline.NormalizedTime))
    ApplyForce=false
    ApplyPosition=true
    ApplyPositionFalloff=false
   )
   &SolveForcesAndVelocity(Clamp Velocity=false, Limit Acceleration=false)
   &ScaleColor(
    Scale Alpha=&Multiply_Float(
     A=0.5
     B=&FloatFromCurve(
      FloatCurve=DataInterface() {propertyValues {Curve {keys [<at 0 0>,<at 0.2 1>,<at 0.6 1>,<at 0.8 0>]}}}
      Scale Curve=$System.TimeOfDayBrightnessMultiplier
     )
    )
    ScaleA=true
    ScaleRGB=true
   )
   &ScaleSpriteSize(Scale Factor=&Vector2DFromFloat(Value=&FloatFromCurve(
      FloatCurve=DataInterface() {propertyValues {Curve {keys [<at 0 1.5 RCIM_Cubic>,<at 0.8 0>]}}}
      Scale Curve=&FN_FrontEndFloatModifier(Gameplay=1, Front End=0.7)
     )))
   &CameraOffset(Camera Offset Amount=50)
   /SetVariables(Particles.Flash size:Vector2f=&Multiply_Vector2D_ByFloat(
     Float=&Multiply_Float(
      A=400
      B=&FloatFromCurve(
       FloatCurve=DataInterface() {propertyValues {Curve {keys [
           {arriveTangent=4.1785, interpMode=RCIM_Cubic, leaveTangent=4.1785, tangentMode=RCTM_User, time=0, value=0}
           {arriveTangent=0.9995, interpMode=RCIM_Cubic, leaveTangent=0.9995, tangentMode=RCTM_User, time=0.1378, value=0.6825}
           {arriveTangent=-0.0405, interpMode=RCIM_Cubic, leaveTangent=-0.0405, tangentMode=RCTM_User, time=0.62, value=1}
          ]}}}
       Scale Curve=&FN_FrontEndFloatModifier(Gameplay=0.9, Front End=0.45)
      )
     )
     Vector2D=<1.5 1>
    )) {isSetParameters=true, module=SetVariables_#12}
   /SetVariables(Particles.Flash color:LinearColor=&MakeLinearColorFromVectorAndFloat(
     "Vector (RGB)"=&MakeVectorFromLinearColorRGB(Color=&SelectLinearColorFromArray(
       Array Sampling Mode=::NewEnumerator1
       Direct Array Index=$Particles.ColorPick
       Color Selection Array=DataInterface() {propertyValues {ColorData [<rgba 1 0.8 0 1>, <rgba 0.3 0.3 0.3 1>, <rgba 1 0.1 0 1>]}}
       Random Seed=?
      ))
     "Float (Alpha)"=&Multiply_Float(
      A=&Multiply_Float(
       A=1
       B=&FloatFromCurve(
        FloatCurve=DataInterface() {propertyValues {Curve {keys [<at 0 0>,<at 0.124 1>,<at 0.4 1>,<at 0.65 0>]}}}
        Scale Curve=$System.TimeOfDayBrightnessMultiplier
       )
      )
      B=1
     )
    )) {isSetParameters=true, module=SetVariables_#13}
   &FN_Sprite_Attributes(
    Color=&MakeLinearColorFromVectorAndFloat(
     "Vector (RGB)"=&MakeVectorFromLinearColorRGB(Color=&SelectLinearColorFromArray(
       Array Sampling Mode=::NewEnumerator1
       Direct Array Index=$Particles.ColorPick
       Color Selection Array=DataInterface() {propertyValues {ColorData [<rgba 0.6 0.2 0 1>, <rgba 0.15 0.15 0.15 1>, <rgba 0.5 0.05 0 1>]}}
       Random Seed=?
      ))
     "Float (Alpha)"=&Multiply_Float(
      A=0.5
      B=&FloatFromCurve(FloatCurve=DataInterface() {propertyValues {Curve {keys [
           {arriveTangent=0.4005, interpMode=RCIM_Cubic, leaveTangent=0.4005, tangentMode=RCTM_User, time=0, value=0}
           {arriveTangent=0.0256, interpMode=RCIM_Cubic, leaveTangent=0.0256, tangentMode=RCTM_User, time=0.2, value=1}
           <at 1 0 RCIM_Cubic>
          ]}}})
     )
    )
    Create Color Attribute=true
    Create SpriteSize Attribute=true
    SpriteSize=&Multiply_Vector2D_ByFloat(
     Float=&MakeFloatFromVector2d(Vector 2D=&Vector2DFromFloat(Value=&FloatFromCurve(
        FloatCurve=DataInterface() {propertyValues {Curve {keys [
            {arriveTangent=0.4346, interpMode=RCIM_Cubic, leaveTangent=0.4346, tangentMode=RCTM_User, time=0, value=0}
            {arriveTangent=0.0473, interpMode=RCIM_Cubic, leaveTangent=0.0473, tangentMode=RCTM_User, time=0.2, value=1}
           ]}}}
        Scale Curve=&FN_FrontEndFloatModifier(Gameplay=1, Front End=0.5)
       )))
     Vector2D=<200 200>
    )
    Write to Module Namespace=true
   )
   &FN_Sprite_Attributes(
    Color=&MakeLinearColorFromVectorAndFloat(
     "Vector (RGB)"=&MakeVectorFromLinearColorRGB(Color=&SelectLinearColorFromArray(
       Array Sampling Mode=::NewEnumerator1
       Direct Array Index=$Particles.ColorPick
       Color Selection Array=DataInterface() {propertyValues {ColorData [<rgba 3 2 0 1>, <rgba 0.5 0.5 0.5 1>, <rgba 3 0.3 0 1>]}}
       Random Seed=?
      ))
     "Float (Alpha)"=&Multiply_Float(
      A=0.2
      B=&FloatFromCurve(FloatCurve=DataInterface() {propertyValues {Curve {keys [
           {arriveTangent=0.4005, interpMode=RCIM_Cubic, leaveTangent=0.4005, tangentMode=RCTM_User, time=0, value=0}
           {arriveTangent=0.0256, interpMode=RCIM_Cubic, leaveTangent=0.0256, tangentMode=RCTM_User, time=0.2, value=1}
           <at 1 0 RCIM_Cubic>
          ]}}})
     )
    )
    Create Color Attribute=true
    Create SpriteSize Attribute=true
    SpriteSize=&Vector2DFromFloat(Value=&Multiply_Float(
      A=120
      B=&FloatFromCurve(
       FloatCurve=DataInterface() {propertyValues {Curve {keys [
           {arriveTangent=0.4346, interpMode=RCIM_Cubic, leaveTangent=0.4346, tangentMode=RCTM_User, time=0, value=0}
           {arriveTangent=0.0473, interpMode=RCIM_Cubic, leaveTangent=0.0473, tangentMode=RCTM_User, time=0.2, value=1}
          ]}}}
       Scale Curve=&FN_FrontEndFloatModifier(Gameplay=1, Front End=0.5)
      )
     ))
    Write to Module Namespace=true
   ) {module=FN_Sprite_Attributes001}
   /SetVariables(
    Particles.LightColor=&MultiplyLinearColor(
     Linear Color=$Particles.Initial.Color
     Scale Factor=&MakeFloat4FromVectorAndFloat(
      "Vector (RGB)"=&VectorFromFloat(Value=5)
      "Float (Alpha)"=&FloatFromCurve(FloatCurve=DataInterface() {propertyValues {Curve {
          defaultValue=0
          keys [
           {arriveTangent=9.0725, interpMode=RCIM_Cubic, leaveTangent=9.0725, tangentMode=RCTM_User, time=0, value=0}
           {arriveTangent=-0.1172, interpMode=RCIM_Cubic, leaveTangent=-0.1172, tangentMode=RCTM_User, time=0.1408, value=0.9847}
           {arriveTangent=-0.0405, interpMode=RCIM_Cubic, leaveTangent=-0.0405, tangentMode=RCTM_User, time=1, value=0}
          ]
         }}})
     )
    )
    Particles.LightRadius=1200
   ) {enabled=false, isSetParameters=true, module=SetVariables_#14}
   &FN_Mesh_Attributes(
    Color=&MultiplyLinearColor(
     Linear Color=&SelectLinearColorFromArray(
      Array Sampling Mode=::NewEnumerator1
      Direct Array Index=$Particles.ColorPick
      Color Selection Array=DataInterface() {propertyValues {ColorData [<rgba 2 1 0 1>, <rgba 0.4 0.4 0.4 1>, <rgba 2 0.2 0 1>]}}
      Random Seed=?
     )
     Scale Factor=&MakeVector4FromVectorAndFloat(
      "Vector (RGB)"=&VectorFromFloat(Value=1)
      "Float (Alpha)"=&FloatFromCurve(
       FloatCurve=DataInterface() {propertyValues {Curve {keys [
           {arriveTangent=-4.9376, interpMode=RCIM_Cubic, leaveTangent=7.1838, tangentMode=RCTM_User, time=0, value=1}
           {arriveTangent=-2.4774, interpMode=RCIM_Cubic, leaveTangent=-2.4774, tangentMode=RCTM_User, time=0.1389, value=1.3591}
           {arriveTangent=-0.8467, interpMode=RCIM_Cubic, leaveTangent=-0.8467, tangentMode=RCTM_User, time=0.4181, value=0.6281}
           {arriveTangent=-2.194, interpMode=RCIM_Cubic, leaveTangent=-2.1946, tangentMode=RCTM_User, time=1, value=0}
          ]}}}
       Scale Curve=2
      )
     )
    )
    Create Color Attribute=true
    Create Scale Attribute=true
    Scale=&Multiply_VectorByFloat(
     Vector=&VectorFromFloat(Value=&FloatFromCurve(
       FloatCurve=DataInterface() {propertyValues {Curve {keys [
           {arriveTangent=4.6803, interpMode=RCIM_Cubic, leaveTangent=4.6803, tangentMode=RCTM_User, time=0, value=0}
           {arriveTangent=1.6276, interpMode=RCIM_Cubic, leaveTangent=1.6276, tangentMode=RCTM_User, time=0.1923, value=0.5575}
           {arriveTangent=0.2011, interpMode=RCIM_Cubic, leaveTangent=0.5033, tangentMode=RCTM_User, time=1, value=1}
          ]}}}
       Scale Curve=&FN_FrontEndFloatModifier(Gameplay=1, Front End=0.5)
      ))
     Float=0.075
    )
    Write to Module Namespace=true
   )
   /SetVariables(Particles.MeshDynamicParam0:Vector4f=&MakeVector4(
     X=1
     Y=1
     Z=&FloatFromCurve(
      FloatCurve=DataInterface() {propertyValues {Curve {keys [
          {arriveTangent=3.0114, interpMode=RCIM_Cubic, leaveTangent=3.0114, tangentMode=RCTM_User, time=0, value=0}
          {arriveTangent=0.9456, interpMode=RCIM_Cubic, leaveTangent=0.9456, tangentMode=RCTM_User, time=0.4343, value=0.7445}
          {arriveTangent=0.3236, interpMode=RCIM_Cubic, leaveTangent=0.7406, tangentMode=RCTM_User, time=1, value=1}
         ]}}}
      Scale Curve=0.2
     )
     W=&FloatFromCurve(
      FloatCurve=DataInterface() {propertyValues {Curve {keys [
          {arriveTangent=3.4321, interpMode=RCIM_Cubic, leaveTangent=3.4321, tangentMode=RCTM_User, time=0, value=1}
          {arriveTangent=-0.0701, interpMode=RCIM_Cubic, leaveTangent=-0.0701, tangentMode=RCTM_User, time=0.1672, value=1.5}
          {arriveTangent=-0.286, interpMode=RCIM_Cubic, leaveTangent=-0.286, tangentMode=RCTM_User, time=1, value=0.75}
         ]}}}
      Scale Curve=1.25
     )
    )) {isSetParameters=true, module=SetVariables_#15}
   /SetVariables(Particles.Position=&Add_Vector(
     A=Linked() {linkedVariable {name=Particles.Initial.Position, type=NiagaraPosition}}
     B=&Lerp_Vector(
      A=<0 0 0>
      B=<0 0 300>
      Alpha=&FloatFromCurve(FloatCurve=DataInterface() {propertyValues {Curve {defaultValue=1, keys [{tangentMode=RCTM_User, time=0, value=0}, {tangentMode=RCTM_User, time=1, value=1}]}}})
     )
    )) {enabled=false, isSetParameters=true, module=SetVariables_#16}
   &DynamicMaterialParameters(
    Index 3 Param 1=&MakeFloatFromVector(
     Vector=&Lerp_Vector(
      A=<0 0 0>
      B=<0 0 300>
      Alpha=&FloatFromCurve(FloatCurve=DataInterface() {propertyValues {Curve {defaultValue=1, keys [{tangentMode=RCTM_User, time=0, value=0}, {tangentMode=RCTM_User, time=1, value=1}]}}})
     )
     Channel=::NewEnumerator0
    )
    Module.Index3Float0Write=true
    Module.Index3Float1Write=false
    Module.Index3Float2Write=false
    Module.Index3Float3Write=false
    Write Parameter Index 0=::NewEnumerator0
    Write Parameter Index 3=::NewEnumerator1
   ) {enabled=false, module=DynamicMaterialParameters001}
   /SetVariables(Particles.Offset:NiagaraFloat=&MakeFloatFromVector(
     Vector=&Lerp_Vector(
      A=<0 0 0>
      B=<0 0 300>
      Alpha=&FloatFromCurve(FloatCurve=DataInterface() {propertyValues {Curve {defaultValue=1, keys [{tangentMode=RCTM_User, time=0, value=0}, {tangentMode=RCTM_User, time=1, value=1}]}}})
     )
     Channel=::NewEnumerator2
    )) {enabled=false, isSetParameters=true, module=SetVariables_#17}
   &FN_Light_Attributes(
    Color=$Particles.Initial.Color
    Create Color Attribute=true
    Create Radius Attribute=true
    Radius=800
    Scale Color=&FloatFromCurve(
     FloatCurve=DataInterface() {propertyValues {Curve {
        defaultValue=0
        keys [
         {arriveTangent=9.0725, interpMode=RCIM_Cubic, leaveTangent=9.0725, tangentMode=RCTM_User, time=0, value=0}
         {arriveTangent=-0.1172, interpMode=RCIM_Cubic, leaveTangent=-0.1172, tangentMode=RCTM_User, time=0.1408, value=0.9847}
         {arriveTangent=-0.0405, interpMode=RCIM_Cubic, leaveTangent=-0.0405, tangentMode=RCTM_User, time=1, value=0}
        ]
       }}}
     Scale Curve=6
    )
   )
   /SetVariables(Particles.AirglowDynamic:Vector4f=<2 5 1 1>) {isSetParameters=true}
   &FN_Sprite_Attributes(
    Color=&MultiplyLinearColor(
     Linear Color=$Particles.Color
     Scale Factor=&MakeVector4FromVectorAndFloat(
      "Vector (RGB)"=<1 1 1>
      "Float (Alpha)"=&FloatFromCurve(
       FloatCurve=DataInterface() {propertyValues {Curve {
          defaultValue=0
          keys [
           {arriveTangent=9.0725, interpMode=RCIM_Cubic, leaveTangent=9.0725, tangentMode=RCTM_User, time=0, value=0}
           {arriveTangent=-0.055, interpMode=RCIM_Cubic, leaveTangent=-0.055, tangentMode=RCTM_User, time=0.1408, value=0.9847}
           {arriveTangent=-1.5458, interpMode=RCIM_Cubic, leaveTangent=-1.5458, tangentMode=RCTM_User, time=0.7331, value=0.7205}
           {arriveTangent=-0.0405, interpMode=RCIM_Cubic, leaveTangent=-0.0405, tangentMode=RCTM_User, time=1, value=0}
          ]
         }}}
       Scale Curve=0.2
      )
     )
    )
    Create Color Attribute=true
    Create SpriteSize Attribute=true
    SpriteSize=&Multiply_Vector2D_ByFloat(
     Float=&FloatFromCurve(FloatCurve=DataInterface() {propertyValues {Curve {
         defaultValue=1
         keys [
          {arriveTangent=8.3978, interpMode=RCIM_Cubic, leaveTangent=8.3978, tangentMode=RCTM_User, time=0, value=0}
          {arriveTangent=0.0224, interpMode=RCIM_Cubic, leaveTangent=0.0224, tangentMode=RCTM_User, time=0.25, value=1}
         ]
        }}})
     Vector2D=<600 600>
    )
    Write to Module Namespace=true
   ) {module=FN_Sprite_Attributes002}
  ]
  rendererClasses [
   &NiagaraSpriteRendererProperties
   &NiagaraLightRendererProperties
   &NiagaraMeshRendererProperties
  ]
  renderers [
   {rendererClass=&NiagaraSpriteRendererProperties}
   {rendererClass=&NiagaraSpriteRendererProperties, rendererIndex=1}
   {rendererClass=&NiagaraSpriteRendererProperties, rendererIndex=2}
   {rendererClass=&NiagaraSpriteRendererProperties, rendererIndex=3}
   {rendererClass=&NiagaraLightRendererProperties, rendererIndex=4}
   {rendererClass=&NiagaraMeshRendererProperties, rendererIndex=5}
   {rendererClass=&NiagaraSpriteRendererProperties, rendererIndex=6}
  ]
 }
 Flare_With_Streak_FE {
  emitterUpdateScript [
   &EmitterState()
   &SpawnRate(
    Spawn Probability=0.3
    SpawnRate=&FN_FrontEndFloatModifier(Gameplay=0, Front End=8)
    Use Spawn Probability=true
   )
   &SpawnBurst_Instantaneous(Use Spawn Probability=false, UseLoopCountLimit=false) {enabled=false}
  ]
  particleSpawnScript [
   /SetVariables(Particles.ColorPick:NiagaraInt32=&RandomRangeInt(Minimum=0, Maximum=2)) {isSetParameters=true, module=SetVariables_#10}
   &InitializeParticle(
    Color=&SelectLinearColorFromArray(
     Array Sampling Mode=::NewEnumerator1
     Direct Array Index=$Particles.ColorPick
     Color Selection Array=DataInterface() {propertyValues {ColorData [<rgba 3 2 0 1>, <rgba 0.5 0.5 0.5 1>, <rgba 3 0.3 0 1>]}}
     Random Seed=?
    )
    Lifetime=0.5
    Position Offset=$System.ArrowTipPosition
    Position Offset Coordinate Space=::Local
    Sprite Rotation Mode=::NewEnumerator2
    Sprite Size Mode=::NewEnumerator3
    Uniform Sprite Size=&Multiply_Float(A=250, B=$System.FE_SizeModifier)
    UsePositionOffset=true
    Write Material Random=true
   )
   &DynamicMaterialParameters(
    Index 0 Param 1=&RandomRangeFloat(Minimum=0.4, Maximum=0.6)
    Index 0 Param 2=1
    Index 0 Param 3=1
    Index 0 Param 4=1
    Module.Index0Float0Write=true
    Module.Index0Float1Write=true
    Module.Index0Float2Write=true
    Module.Index0Float3Write=true
   )
   /SetVariables() {isSetParameters=true, module=SetVariables_#11}
   &ShapeLocation(
    Cylinder Height=&FN_FrontEndFloatModifier(Gameplay=300, Front End=300)
    Cylinder Radius=&FN_FrontEndFloatModifier(Gameplay=500, Front End=150)
    MaskSpawnGroup=false
    Offset=&FN_FrontEndVector3Modifier(Gameplay=<0 0 -600>, Front End=<0 0 25>)
    Offset Mode=::NewEnumerator0
    Override Seed=false
    Shape Primitive=::NewEnumerator1
    Surface Only=true
    Surface Only Band Thickness=0.35
    Use Endcaps In Surface Only Mode=false
    UseOwnerScale=true
   )
   &AddVelocity(
    Velocity=<0 0 1>
    Velocity Speed Scale=&FN_FrontEndFloatModifier(
     Gameplay=0
     Front End=&Multiply_Float(A=&Lerp_Float(A=0.5, B=1, Alpha=$System.Timeline.NormalizedTime), B=500)
    )
   ) {enabled=false}
  ]
  particleUpdateScript [
   &ParticleState()
   &InheritSourceMovement(
    Applied Position Scale=&Multiply_VectorByFloat(Vector=<1 1 1>, Float=&Lerp_Float(A=0.5, B=0.8, Alpha=$System.Timeline.NormalizedTime))
    ApplyForce=false
    ApplyPosition=true
    ApplyPositionFalloff=false
   ) {enabled=false}
   &SolveForcesAndVelocity(Clamp Velocity=false, Limit Acceleration=false)
   &ScaleColor(
    Scale Alpha=&Multiply_Float(
     A=0.5
     B=&FloatFromCurve(
      FloatCurve=DataInterface() {propertyValues {Curve {keys [<at 0 0>,<at 0.2 1>,<at 0.6 1>,<at 0.8 0>]}}}
      Scale Curve=$System.TimeOfDayBrightnessMultiplier
     )
    )
    ScaleA=true
    ScaleRGB=true
   )
   &ScaleSpriteSize(Scale Factor=&Vector2DFromFloat(Value=&FloatFromCurve(
      FloatCurve=DataInterface() {propertyValues {Curve {keys [<at 0 1.5 RCIM_Cubic>,<at 0.8 0>]}}}
      Scale Curve=&FN_FrontEndFloatModifier(Gameplay=1, Front End=0.7)
     )))
   &CameraOffset(Camera Offset Amount=$System.FE_CameraOffset)
   /SetVariables(Particles.Flash size:Vector2f=&Multiply_Vector2D_ByFloat(
     Float=&Multiply_Float(
      A=400
      B=&FloatFromCurve(
       FloatCurve=DataInterface() {propertyValues {Curve {keys [
           {arriveTangent=4.1785, interpMode=RCIM_Cubic, leaveTangent=4.1785, tangentMode=RCTM_User, time=0, value=0}
           {arriveTangent=0.9995, interpMode=RCIM_Cubic, leaveTangent=0.9995, tangentMode=RCTM_User, time=0.1378, value=0.6825}
           {arriveTangent=-0.0405, interpMode=RCIM_Cubic, leaveTangent=-0.0405, tangentMode=RCTM_User, time=0.62, value=1}
          ]}}}
       Scale Curve=&FN_FrontEndFloatModifier(Gameplay=0.9, Front End=&Multiply_Float(A=0.45, B=$System.FE_SizeModifier))
      )
     )
     Vector2D=<1.5 1>
    )) {isSetParameters=true, module=SetVariables_#12}
   /SetVariables(Particles.Flash color:LinearColor=&MakeLinearColorFromVectorAndFloat(
     "Vector (RGB)"=&MakeVectorFromLinearColorRGB(Color=&SelectLinearColorFromArray(
       Array Sampling Mode=::NewEnumerator1
       Direct Array Index=$Particles.ColorPick
       Color Selection Array=DataInterface() {propertyValues {ColorData [<rgba 1 0.8 0 1>, <rgba 0.3 0.3 0.3 1>, <rgba 1 0.1 0 1>]}}
       Random Seed=?
      ))
     "Float (Alpha)"=&Multiply_Float(
      A=&Multiply_Float(
       A=1
       B=&FloatFromCurve(
        FloatCurve=DataInterface() {propertyValues {Curve {keys [<at 0 0>,<at 0.124 1>,<at 0.4 1>,<at 0.65 0>]}}}
        Scale Curve=$System.TimeOfDayBrightnessMultiplier
       )
      )
      B=1
     )
    )) {isSetParameters=true, module=SetVariables_#13}
   &FN_Sprite_Attributes(
    Color=&MakeLinearColorFromVectorAndFloat(
     "Vector (RGB)"=&MakeVectorFromLinearColorRGB(Color=&SelectLinearColorFromArray(
       Array Sampling Mode=::NewEnumerator1
       Direct Array Index=$Particles.ColorPick
       Color Selection Array=DataInterface() {propertyValues {ColorData [<rgba 0.6 0.2 0 1>, <rgba 0.15 0.15 0.15 1>, <rgba 0.5 0.05 0 1>]}}
       Random Seed=?
      ))
     "Float (Alpha)"=&Multiply_Float(
      A=0.5
      B=&FloatFromCurve(FloatCurve=DataInterface() {propertyValues {Curve {keys [
           {arriveTangent=0.4005, interpMode=RCIM_Cubic, leaveTangent=0.4005, tangentMode=RCTM_User, time=0, value=0}
           {arriveTangent=0.0256, interpMode=RCIM_Cubic, leaveTangent=0.0256, tangentMode=RCTM_User, time=0.2, value=1}
           <at 1 0 RCIM_Cubic>
          ]}}})
     )
    )
    Create Color Attribute=true
    Create SpriteSize Attribute=true
    SpriteSize=&Multiply_Vector2D_ByFloat(
     Float=&MakeFloatFromVector2d(Vector 2D=&Vector2DFromFloat(Value=&FloatFromCurve(
        FloatCurve=DataInterface() {propertyValues {Curve {keys [
            {arriveTangent=0.4346, interpMode=RCIM_Cubic, leaveTangent=0.4346, tangentMode=RCTM_User, time=0, value=0}
            {arriveTangent=0.0473, interpMode=RCIM_Cubic, leaveTangent=0.0473, tangentMode=RCTM_User, time=0.2, value=1}
           ]}}}
        Scale Curve=&FN_FrontEndFloatModifier(Gameplay=1, Front End=&Multiply_Float(A=0.5, B=$System.FE_SizeModifier))
       )))
     Vector2D=<200 200>
    )
    Write to Module Namespace=true
   )
   &FN_Sprite_Attributes(
    Color=&MakeLinearColorFromVectorAndFloat(
     "Vector (RGB)"=&MakeVectorFromLinearColorRGB(Color=&SelectLinearColorFromArray(
       Array Sampling Mode=::NewEnumerator1
       Direct Array Index=$Particles.ColorPick
       Color Selection Array=DataInterface() {propertyValues {ColorData [<rgba 3 2 0 1>, <rgba 0.5 0.5 0.5 1>, <rgba 3 0.3 0 1>]}}
       Random Seed=?
      ))
     "Float (Alpha)"=&Multiply_Float(
      A=0.2
      B=&FloatFromCurve(FloatCurve=DataInterface() {propertyValues {Curve {keys [
           {arriveTangent=0.4005, interpMode=RCIM_Cubic, leaveTangent=0.4005, tangentMode=RCTM_User, time=0, value=0}
           {arriveTangent=0.0256, interpMode=RCIM_Cubic, leaveTangent=0.0256, tangentMode=RCTM_User, time=0.2, value=1}
           <at 1 0 RCIM_Cubic>
          ]}}})
     )
    )
    Create Color Attribute=true
    Create SpriteSize Attribute=true
    SpriteSize=&Vector2DFromFloat(Value=&Multiply_Float(
      A=120
      B=&FloatFromCurve(
       FloatCurve=DataInterface() {propertyValues {Curve {keys [
           {arriveTangent=0.4346, interpMode=RCIM_Cubic, leaveTangent=0.4346, tangentMode=RCTM_User, time=0, value=0}
           {arriveTangent=0.0473, interpMode=RCIM_Cubic, leaveTangent=0.0473, tangentMode=RCTM_User, time=0.2, value=1}
          ]}}}
       Scale Curve=&FN_FrontEndFloatModifier(Gameplay=1, Front End=&Multiply_Float(A=0.5, B=$System.FE_SizeModifier))
      )
     ))
    Write to Module Namespace=true
   ) {module=FN_Sprite_Attributes001}
   /SetVariables(
    Particles.LightColor=&MultiplyLinearColor(
     Linear Color=$Particles.Initial.Color
     Scale Factor=&MakeFloat4FromVectorAndFloat(
      "Vector (RGB)"=&VectorFromFloat(Value=5)
      "Float (Alpha)"=&FloatFromCurve(FloatCurve=DataInterface() {propertyValues {Curve {
          defaultValue=0
          keys [
           {arriveTangent=9.0725, interpMode=RCIM_Cubic, leaveTangent=9.0725, tangentMode=RCTM_User, time=0, value=0}
           {arriveTangent=-0.1172, interpMode=RCIM_Cubic, leaveTangent=-0.1172, tangentMode=RCTM_User, time=0.1408, value=0.9847}
           {arriveTangent=-0.0405, interpMode=RCIM_Cubic, leaveTangent=-0.0405, tangentMode=RCTM_User, time=1, value=0}
          ]
         }}})
     )
    )
    Particles.LightRadius=1200
   ) {enabled=false, isSetParameters=true, module=SetVariables_#14}
   &FN_Mesh_Attributes(
    Color=&MultiplyLinearColor(
     Linear Color=&SelectLinearColorFromArray(
      Array Sampling Mode=::NewEnumerator1
      Direct Array Index=$Particles.ColorPick
      Color Selection Array=DataInterface() {propertyValues {ColorData [<rgba 2 1 0 1>, <rgba 0.4 0.4 0.4 1>, <rgba 2 0.2 0 1>]}}
      Random Seed=?
     )
     Scale Factor=&MakeVector4FromVectorAndFloat(
      "Vector (RGB)"=&VectorFromFloat(Value=1)
      "Float (Alpha)"=&FloatFromCurve(
       FloatCurve=DataInterface() {propertyValues {Curve {keys [
           {arriveTangent=-4.9376, interpMode=RCIM_Cubic, leaveTangent=7.1838, tangentMode=RCTM_User, time=0, value=1}
           {arriveTangent=-2.4774, interpMode=RCIM_Cubic, leaveTangent=-2.4774, tangentMode=RCTM_User, time=0.1389, value=1.3591}
           {arriveTangent=-0.8467, interpMode=RCIM_Cubic, leaveTangent=-0.8467, tangentMode=RCTM_User, time=0.4181, value=0.6281}
           {arriveTangent=-2.194, interpMode=RCIM_Cubic, leaveTangent=-2.1946, tangentMode=RCTM_User, time=1, value=0}
          ]}}}
       Scale Curve=2
      )
     )
    )
    Create Color Attribute=true
    Create Scale Attribute=true
    Scale=&Multiply_VectorByFloat(
     Vector=&VectorFromFloat(Value=&FloatFromCurve(
       FloatCurve=DataInterface() {propertyValues {Curve {keys [
           {arriveTangent=4.6803, interpMode=RCIM_Cubic, leaveTangent=4.6803, tangentMode=RCTM_User, time=0, value=0}
           {arriveTangent=1.6276, interpMode=RCIM_Cubic, leaveTangent=1.6276, tangentMode=RCTM_User, time=0.1923, value=0.5575}
           {arriveTangent=0.2011, interpMode=RCIM_Cubic, leaveTangent=0.5033, tangentMode=RCTM_User, time=1, value=1}
          ]}}}
       Scale Curve=&Multiply_Float(A=0.5, B=$System.FE_SizeModifier)
      ))
     Float=0.075
    )
    Write to Module Namespace=true
   )
   /SetVariables(Particles.MeshDynamicParam0:Vector4f=&MakeVector4(
     X=1
     Y=1
     Z=&FloatFromCurve(
      FloatCurve=DataInterface() {propertyValues {Curve {keys [
          {arriveTangent=3.0114, interpMode=RCIM_Cubic, leaveTangent=3.0114, tangentMode=RCTM_User, time=0, value=0}
          {arriveTangent=0.9456, interpMode=RCIM_Cubic, leaveTangent=0.9456, tangentMode=RCTM_User, time=0.4343, value=0.7445}
          {arriveTangent=0.3236, interpMode=RCIM_Cubic, leaveTangent=0.7406, tangentMode=RCTM_User, time=1, value=1}
         ]}}}
      Scale Curve=0.1
     )
     W=&FloatFromCurve(
      FloatCurve=DataInterface() {propertyValues {Curve {keys [
          {arriveTangent=3.4321, interpMode=RCIM_Cubic, leaveTangent=3.4321, tangentMode=RCTM_User, time=0, value=1}
          {arriveTangent=-0.0701, interpMode=RCIM_Cubic, leaveTangent=-0.0701, tangentMode=RCTM_User, time=0.1672, value=1.5}
          {arriveTangent=-0.286, interpMode=RCIM_Cubic, leaveTangent=-0.286, tangentMode=RCTM_User, time=1, value=0.75}
         ]}}}
      Scale Curve=0.5
     )
    )) {isSetParameters=true, module=SetVariables_#15}
   /SetVariables(Particles.Position=&Add_Vector(
     A=Linked() {linkedVariable {name=Particles.Initial.Position, type=NiagaraPosition}}
     B=&Lerp_Vector(
      A=<0 0 0>
      B=<0 0 300>
      Alpha=&FloatFromCurve(FloatCurve=DataInterface() {propertyValues {Curve {defaultValue=1, keys [{tangentMode=RCTM_User, time=0, value=0}, {tangentMode=RCTM_User, time=1, value=1}]}}})
     )
    )) {enabled=false, isSetParameters=true, module=SetVariables_#16}
   &DynamicMaterialParameters(
    Index 3 Param 1=&MakeFloatFromVector(
     Vector=&Lerp_Vector(
      A=<0 0 0>
      B=<0 0 300>
      Alpha=&FloatFromCurve(FloatCurve=DataInterface() {propertyValues {Curve {defaultValue=1, keys [{tangentMode=RCTM_User, time=0, value=0}, {tangentMode=RCTM_User, time=1, value=1}]}}})
     )
     Channel=::NewEnumerator0
    )
    Module.Index3Float0Write=true
    Module.Index3Float1Write=false
    Module.Index3Float2Write=false
    Module.Index3Float3Write=false
    Write Parameter Index 0=::NewEnumerator0
    Write Parameter Index 3=::NewEnumerator1
   ) {enabled=false, module=DynamicMaterialParameters001}
   /SetVariables(Particles.Offset:NiagaraFloat=&MakeFloatFromVector(
     Vector=&Lerp_Vector(
      A=<0 0 0>
      B=<0 0 300>
      Alpha=&FloatFromCurve(FloatCurve=DataInterface() {propertyValues {Curve {defaultValue=1, keys [{tangentMode=RCTM_User, time=0, value=0}, {tangentMode=RCTM_User, time=1, value=1}]}}})
     )
     Channel=::NewEnumerator2
    )) {enabled=false, isSetParameters=true, module=SetVariables_#17}
   &FN_Light_Attributes(
    Color=$Particles.Initial.Color
    Create Color Attribute=true
    Create Radius Attribute=true
    Radius=&Multiply_Float(A=400, B=$System.FE_SizeModifier)
    Scale Color=&FloatFromCurve(
     FloatCurve=DataInterface() {propertyValues {Curve {
        defaultValue=0
        keys [
         {arriveTangent=9.0725, interpMode=RCIM_Cubic, leaveTangent=9.0725, tangentMode=RCTM_User, time=0, value=0}
         {arriveTangent=-0.1172, interpMode=RCIM_Cubic, leaveTangent=-0.1172, tangentMode=RCTM_User, time=0.1408, value=0.9847}
         {arriveTangent=-0.0405, interpMode=RCIM_Cubic, leaveTangent=-0.0405, tangentMode=RCTM_User, time=1, value=0}
        ]
       }}}
     Scale Curve=6
    )
   )
   /SetVariables(Particles.Offset:Vector3f=&Lerp_Vector(A=<0 0 6>, B=<0 0 12>, Alpha=$System.Timeline.NormalizedTime)) {isSetParameters=true}
   /SetVariables(Particles.FEPosition:NiagaraPosition=&Lerp_Vector(
     A=Linked() {linkedVariable {name=Particles.Position, type=NiagaraPosition}}
     B=&AddVectorToPosition(Position=$Particles.Position, Vector=$Particles.Offset)
     Alpha=&FloatFromCurve(FloatCurve=DataInterface() {propertyValues {Curve {defaultValue=1, keys [{tangentMode=RCTM_User, time=0, value=0}, {tangentMode=RCTM_User, time=1, value=1}]}}})
    )) {enabled=false, isSetParameters=true}
   /SetVariables(Particles.Position=&AddVectorToPosition(Position=$Particles.Position, Vector=$System.Offset)) {isSetParameters=true}
   &FN_Sprite_Attributes(
    Color=&MultiplyLinearColor(
     Linear Color=$Particles.Color
     Scale Factor=&MakeVector4FromVectorAndFloat(
      "Vector (RGB)"=<1 1 1>
      "Float (Alpha)"=&FloatFromCurve(
       FloatCurve=DataInterface() {propertyValues {Curve {
          defaultValue=0
          keys [
           {arriveTangent=9.0725, interpMode=RCIM_Cubic, leaveTangent=9.0725, tangentMode=RCTM_User, time=0, value=0}
           {arriveTangent=-0.055, interpMode=RCIM_Cubic, leaveTangent=-0.055, tangentMode=RCTM_User, time=0.1408, value=0.9847}
           {arriveTangent=-1.5458, interpMode=RCIM_Cubic, leaveTangent=-1.5458, tangentMode=RCTM_User, time=0.7331, value=0.7205}
           {arriveTangent=-0.0405, interpMode=RCIM_Cubic, leaveTangent=-0.0405, tangentMode=RCTM_User, time=1, value=0}
          ]
         }}}
       Scale Curve=0.2
      )
     )
    )
    Create Color Attribute=true
    Create SpriteSize Attribute=true
    SpriteSize=&Multiply_Vector2D_ByFloat(
     Float=&FloatFromCurve(FloatCurve=DataInterface() {propertyValues {Curve {
         defaultValue=1
         keys [
          {arriveTangent=8.3978, interpMode=RCIM_Cubic, leaveTangent=8.3978, tangentMode=RCTM_User, time=0, value=0}
          {arriveTangent=0.0224, interpMode=RCIM_Cubic, leaveTangent=0.0224, tangentMode=RCTM_User, time=0.25, value=1}
         ]
        }}})
     Vector2D=<600 600>
    )
    Write to Module Namespace=true
   ) {module=FN_Sprite_Attributes002}
   /SetVariables(Particles.AirglowDynamic:Vector4f=<2 5 1 1>) {isSetParameters=true}
  ]
  rendererClasses [
   &NiagaraSpriteRendererProperties
   &NiagaraLightRendererProperties
   &NiagaraMeshRendererProperties
  ]
  renderers [
   {rendererClass=&NiagaraSpriteRendererProperties}
   {rendererClass=&NiagaraSpriteRendererProperties, rendererIndex=1}
   {rendererClass=&NiagaraSpriteRendererProperties, rendererIndex=2}
   {rendererClass=&NiagaraSpriteRendererProperties, rendererIndex=3}
   {rendererClass=&NiagaraLightRendererProperties, rendererIndex=4}
   {rendererClass=&NiagaraMeshRendererProperties, rendererIndex=5}
   {rendererClass=&NiagaraSpriteRendererProperties, rendererIndex=6}
  ]
 }
 Spark_High {
  emitterUpdateScript [
   &EmitterState()
   &SpawnBurst_Instantaneous(Spawn Count=333, Use Spawn Probability=false, UseLoopCountLimit=false) {enabled=false}
   &SpawnParticlesFromOtherEmitter(
    Attribute Reader=DataInterface() {propertyValues {EmitterBinding {bindingMode=Other, emitterName=Flare_With_Streak}}}
    Maximum Age Spawn Threshold=0
    Spawn Rate=&FN_FrontEndFloatModifier(Gameplay=&Lerp_Float(A=200, B=100, Alpha=$System.Timeline.NormalizedTime), Front End=0)
    Use Maximum Age=false
    Use Minimum Age=false
   )
  ]
  particleSpawnScript [
   &InitializeParticle(
    AdjustAlpha=false
    AdjustHue=true
    AdjustSaturation=false
    AdjustValue=false
    Color=<rgba 0 9.9 33 1>
    Color Mode=::NewEnumerator3
    Hue Shift Range:Vector2f=<-0.2 0>
    Lifetime=&Multiply_Float(
     A=&RandomRangeFloat(Minimum=0.25, Maximum=1.3)
     B=&FN_FrontEndFloatModifier(Gameplay=1, Front End=0.65)
    )
    Position Mode=::NewEnumerator0
    Sprite Size Mode=::NewEnumerator1
    Uniform Sprite Size Max=12
    Uniform Sprite Size Min=9
    Write Material Random=true
   )
   &SampleParticlesFromOtherEmitter(Maximum Sampled Particle Age=0.1, Module.MaxSampledAgeCutoff=true, Module.MinSampledAgeCutoff=false, Reject Particle By Sampled Age=true)
   &ShapeLocation(MaskSpawnGroup=false, Override Seed=false, Sphere Radius=22, UseOwnerScale=true)
   &CurlNoiseForce(
    Noise Frequency=99
    Noise Strength=&FN_FrontEndFloatModifier(Gameplay=666, Front End=50)
    Pan Noise=true
   )
   &ApplyInitialForces()
   &AddVelocity(
    Module.Use Falloff Distance=false
    Override Seed=false
    Velocity Mode=::NewEnumerator1
    Velocity Origin=$Output.SampleParticlesFromOtherEmitter.SampledPosition
    Velocity Speed=&Multiply_Float(
     A=&FN_FrontEndFloatModifier(Gameplay=1, Front End=0.2)
     B=&RandomRangeFloat(Minimum=100, Maximum=900)
    )
   )
   &self:GetColorPickAttr(
    Particle Attribute Reader=DataInterface() {propertyValues {EmitterBinding {bindingMode=Other, emitterName=Flare_With_Streak}}}
    Particle Index:NiagaraInt32=&MakeIntFromFloat(Float=$Particles.SampleParticlesFromOtherEmitter.SampledIndex)
   ) {module=GetColorPickAttr}
  ]
  particleUpdateScript [
   &ParticleState()
   &GravityForce(Gravity=&RandomRangeVector(Minimum=<0 0 0>, Maximum=<0 0 -100>)) {enabled=false}
   &Drag(
    Drag=&Multiply_Float(
     A=&RandomRangeFloat(Minimum=6, Maximum=11)
     B=&FN_FrontEndFloatModifier(Gameplay=1, Front End=0)
    )
    Use Linear Drag=true
    Use Rotational Drag=true
   )
   &InheritSourceMovement(
    Applied Position Scale=&Multiply_VectorByFloat(Vector=<1 1 1>, Float=&Lerp_Float(A=0.5, B=0.8, Alpha=$System.Timeline.NormalizedTime))
    ApplyForce=false
    ApplyPosition=true
    ApplyPositionFalloff=false
   )
   &SolveForcesAndVelocity(Clamp Velocity=false, Limit Acceleration=false)
   &ScaleSpriteSize(
    Uniform Curve Scale=&FN_FrontEndFloatModifier(Gameplay=1, Front End=0.5)
    Uniform Curve Sprite Scale=DataInterface() {propertyValues {Curve {keys [{tangentMode=RCTM_User, time=0, value=1}, {tangentMode=RCTM_User, time=1, value=0}]}}}
   )
   &ScaleSpriteSizeBySpeed(
    Max Scale Factor=<0.2 5>
    Min Scale Factor=<1 0.5>
    Velocity Threshold:NiagaraFloat=&FN_FrontEndFloatModifier(Gameplay=1000, Front End=400)
   )
   &Color(Color=&HueShiftLinearColor(Hue=&RandomRangeFloat(Minimum=-0.2, Maximum=0), Saturation=1, Value=1, Linear Color=$Particles.Color)) {enabled=false}
   /SetVariables(Particles.Color=&HueShiftLinearColor(
     Hue=&RandomRangeFloat(
      Minimum=&SelectFloatFromArray(Direct Array Index=$Particles.ColorPick, Float Selection Array=DataInterface() {propertyValues {FloatData [-0.2,0,-0.2]}})
      Maximum=&SelectFloatFromArray(Direct Array Index=$Particles.ColorPick, Float Selection Array=DataInterface() {propertyValues {FloatData [0.2,0,0.2]}})
     )
     Saturation=1
     Value=1
     Linear Color=&SelectLinearColorFromArray(
      Array Sampling Mode=::NewEnumerator1
      Direct Array Index=$Particles.ColorPick
      Color Selection Array=DataInterface() {propertyValues {ColorData [<rgba 3 2 0 1>, <rgba 0.5 0.5 0.5 1>, <rgba 3 0.3 0 1>]}}
      Random Seed=?
     )
    )) {enabled=false, isSetParameters=true, module=SetVariables_#18}
   /SetVariables(Particles.Color=&SelectLinearColorFromArray(
     Array Sampling Mode=::NewEnumerator1
     Direct Array Index=$Particles.ColorPick
     Color Selection Array=DataInterface() {propertyValues {ColorData [<rgba 6 4 0 1>, <rgba 1 1 1 1>, <rgba 6 0.6 0 1>]}}
     Random Seed=?
    )) {isSetParameters=true, module=SetVariables_#19}
   &ScaleColor(
    Color Value To Scale=$Particles.Color
    Scale Alpha=&FloatFromCurve(FloatCurve=DataInterface() {propertyValues {Curve {keys [<at 0 1>,<at 0.3 0.25>,<at 1 0>]}}})
    Scale RGB=&Multiply_VectorByFloat(Vector=<1 1 1>, Float=$System.TimeOfDayBrightnessMultiplier)
    ScaleA=true
    ScaleRGB=true
   ) {enabled=false}
  ]
  rendererClasses [&NiagaraSpriteRendererProperties]
  renderers [{rendererClass=&NiagaraSpriteRendererProperties}]
 }
 Spark_High_FE {
  bEnabled=false
  emitterUpdateScript [
   &EmitterState()
   &SpawnBurst_Instantaneous(Spawn Count=333, Use Spawn Probability=false, UseLoopCountLimit=false) {enabled=false}
   &SpawnParticlesFromOtherEmitter(
    Attribute Reader=DataInterface() {propertyValues {EmitterBinding {bindingMode=Other, emitterName=Flare_With_Streak_FE}}}
    Maximum Age Spawn Threshold=0
    Spawn Rate=&FN_FrontEndFloatModifier(Gameplay=0, Front End=50)
    Spawn Rate Per Particle Cap=2000
    Use Maximum Age=false
    Use Minimum Age=false
   )
  ]
  particleSpawnScript [
   &InitializeParticle(
    AdjustAlpha=false
    AdjustHue=true
    AdjustSaturation=false
    AdjustValue=false
    Color=<rgba 0 9.9 33 1>
    Color Mode=::NewEnumerator3
    Hue Shift Range:Vector2f=<-0.2 0>
    Lifetime=&Multiply_Float(
     A=&RandomRangeFloat(Minimum=0.25, Maximum=1.3)
     B=&FN_FrontEndFloatModifier(Gameplay=1, Front End=0.65)
    )
    Position Mode=::NewEnumerator0
    Sprite Size Mode=::NewEnumerator1
    Uniform Sprite Size Max=12
    Uniform Sprite Size Min=9
    Write Material Random=true
   )
   &SampleParticlesFromOtherEmitter(Maximum Sampled Particle Age=0.1, Module.MaxSampledAgeCutoff=true, Module.MinSampledAgeCutoff=false, Reject Particle By Sampled Age=true)
   &ShapeLocation(MaskSpawnGroup=false, Override Seed=false, Sphere Radius=22, UseOwnerScale=true)
   &CurlNoiseForce(
    Noise Frequency=99
    Noise Strength=&FN_FrontEndFloatModifier(Gameplay=666, Front End=50)
    Pan Noise=true
   )
   &ApplyInitialForces()
   &AddVelocity(
    Module.Use Falloff Distance=false
    Override Seed=false
    Velocity Mode=::NewEnumerator1
    Velocity Origin=$Output.SampleParticlesFromOtherEmitter.SampledPosition
    Velocity Speed=&Multiply_Float(
     A=&FN_FrontEndFloatModifier(Gameplay=1, Front End=&Multiply_Float(A=0.1, B=$System.FE_SizeModifier))
     B=&RandomRangeFloat(Minimum=100, Maximum=900)
    )
   )
   &self:GetColorPickAttr(
    Particle Attribute Reader=DataInterface() {propertyValues {EmitterBinding {bindingMode=Other, emitterName=Flare_With_Streak_FE}}}
    Particle Index:NiagaraInt32=&MakeIntFromFloat(Float=$Particles.SampleParticlesFromOtherEmitter.SampledIndex)
   ) {module=GetColorPickAttr}
  ]
  particleUpdateScript [
   &ParticleState()
   &GravityForce(Gravity=&RandomRangeVector(Minimum=<0 0 0>, Maximum=<0 0 -100>)) {enabled=false}
   &Drag(
    Drag=&Multiply_Float(
     A=&RandomRangeFloat(Minimum=6, Maximum=11)
     B=&FN_FrontEndFloatModifier(Gameplay=1, Front End=0)
    )
    Use Linear Drag=true
    Use Rotational Drag=true
   )
   &InheritSourceMovement(
    Applied Position Scale=&Multiply_VectorByFloat(Vector=<1 1 1>, Float=&Lerp_Float(A=0.5, B=0.8, Alpha=$System.Timeline.NormalizedTime))
    ApplyForce=false
    ApplyPosition=true
    ApplyPositionFalloff=false
   )
   &SolveForcesAndVelocity(Clamp Velocity=false, Limit Acceleration=false)
   &ScaleSpriteSize(
    Uniform Curve Scale=&FN_FrontEndFloatModifier(Gameplay=1, Front End=&Multiply_Float(A=0.5, B=$System.FE_SizeModifier))
    Uniform Curve Sprite Scale=DataInterface() {propertyValues {Curve {keys [{tangentMode=RCTM_User, time=0, value=1}, {tangentMode=RCTM_User, time=1, value=0}]}}}
   )
   &ScaleSpriteSizeBySpeed(
    Max Scale Factor=<0.2 3>
    Min Scale Factor=<0.5 0.5>
    Velocity Threshold:NiagaraFloat=&FN_FrontEndFloatModifier(Gameplay=1000, Front End=&Multiply_Float(A=500, B=$System.FE_SizeModifier))
   )
   &Color(Color=&HueShiftLinearColor(Hue=&RandomRangeFloat(Minimum=-0.2, Maximum=0), Saturation=1, Value=1, Linear Color=$Particles.Color)) {enabled=false}
   /SetVariables(Particles.Color=&HueShiftLinearColor(
     Hue=&RandomRangeFloat(
      Minimum=&SelectFloatFromArray(Direct Array Index=$Particles.ColorPick, Float Selection Array=DataInterface() {propertyValues {FloatData [-0.2,0,-0.2]}})
      Maximum=&SelectFloatFromArray(Direct Array Index=$Particles.ColorPick, Float Selection Array=DataInterface() {propertyValues {FloatData [0.2,0,0.2]}})
     )
     Saturation=1
     Value=1
     Linear Color=&SelectLinearColorFromArray(
      Array Sampling Mode=::NewEnumerator1
      Direct Array Index=$Particles.ColorPick
      Color Selection Array=DataInterface() {propertyValues {ColorData [<rgba 3 2 0 1>, <rgba 0.5 0.5 0.5 1>, <rgba 3 0.3 0 1>]}}
      Random Seed=?
     )
    )) {enabled=false, isSetParameters=true, module=SetVariables_#18}
   /SetVariables(Particles.Color=&SelectLinearColorFromArray(
     Array Sampling Mode=::NewEnumerator1
     Direct Array Index=$Particles.ColorPick
     Color Selection Array=DataInterface() {propertyValues {ColorData [<rgba 6 4 0 1>, <rgba 1 1 1 1>, <rgba 6 0.6 0 1>]}}
     Random Seed=?
    )) {isSetParameters=true, module=SetVariables_#19}
   &ScaleColor(
    Color Value To Scale=$Particles.Color
    Scale Alpha=&FloatFromCurve(FloatCurve=DataInterface() {propertyValues {Curve {keys [<at 0 1>,<at 0.3 0.25>,<at 1 0>]}}})
    Scale RGB=&Multiply_VectorByFloat(Vector=<1 1 1>, Float=$System.TimeOfDayBrightnessMultiplier)
    ScaleA=true
    ScaleRGB=true
   ) {enabled=false}
   /SetVariables(Particles.Offset:Vector3f=$System.Offset) {isSetParameters=true}
   /SetVariables(Particles.Position=&AddVectorToPosition(Position=$Particles.Position, Vector=$Particles.Offset)) {isSetParameters=true}
   &CameraOffset(Camera Offset Amount=$System.FE_CameraOffset)
  ]
  rendererClasses [&NiagaraSpriteRendererProperties]
  renderers [{rendererClass=&NiagaraSpriteRendererProperties}]
 }
 TrailEmbers {
  emitterUpdateScript [
   &EmitterState()
   &SpawnParticlesFromOtherEmitter(
    Attribute Reader=DataInterface() {propertyValues {EmitterBinding {bindingMode=Other, emitterName=FireworkFlare_Parent_Sprite}}}
    Maximum Age Spawn Threshold=15
    Minimum Age Spawn Threshold=0
    Spawn Rate=&Multiply_Float(
     A=&SelectFloatFromArray(Direct Array Index=$Engine.QualityLevel, Float Selection Array=DataInterface() {propertyValues {FloatData [10,15,20,40,40]}})
     B=&FN_FrontEndFloatModifier(Gameplay=0.25, Front End=0)
    )
    Spawn Rate Per Particle Cap=125
    Use Maximum Age=false
    Use Minimum Age=false
   )
  ]
  particleSpawnScript [
   &PartitionParticles(Number Of Partitions=4, Partition Mode:ENiagara_PartitionMode=::NewEnumerator5)
   &InitializeParticle(
    Color Mode=::NewEnumerator0
    Lifetime Max=&MergeHelper_FloatSwitch(0=0.3, 1=1, 2=?, 3=?, Index=$Particles.PartitionParticles.Partition, Num=::NewEnumerator0)
    Lifetime Min:NiagaraFloat=&MergeHelper_FloatSwitch(0=0.2, 1=0.7, 2=?, 3=?, Index=$Particles.PartitionParticles.Partition, Num=::NewEnumerator0)
    Lifetime Mode=::NewEnumerator1
    Sprite Rotation Mode=::NewEnumerator2
    Sprite Size Mode=::NewEnumerator3
    Uniform Sprite Size=&MergeHelper_FloatSwitch(0=0.75, 1=3, 2=?, 3=?, Index=$Particles.PartitionParticles.Partition, Num=::NewEnumerator0)
    UsePositionOffset=false
    Write Material Random=true
   )
   &SampleParticlesFromOtherEmitter(
    Applied Velocity Scale:NiagaraFloat=&FloatFromCurve(
     FloatCurve=DataInterface() {propertyValues {Curve {keys [{arriveTangent=-1.801, interpMode=RCIM_Cubic, leaveTangent=-1.801, tangentMode=RCTM_User, time=0, value=1}, <at 1 0>]}}}
     Scale Curve=0.5
    )
    Apply Sampled Alpha=false
    Color Sampling=::NewEnumerator0
    Velocity Sampling=::NewEnumerator0
   )
   &SphereLocation(
    MaskSpawnGroup=false
    Override Seed=false
    Sphere Radius=2
    Surface Only=false
    UseNonUniformScale=false
    UseOffset=false
    UseSphereOrientation=false
   )
  ]
  particleUpdateScript [
   &ParticleState()
   &Drag(Drag=&RandomRangeFloat(Minimum=4, Maximum=6), Use Linear Drag=true, Use Rotational Drag=false)
   &CurlNoiseForce(
    Noise Frequency=&RandomRangeFloat(Minimum=0.25, Maximum=1)
    Noise Strength=&RandomRangeFloat(Minimum=100, Maximum=200)
    Pan Noise=true
    Pan Noise Field=&RandomRangeVector(Minimum=<0 0 0>, Maximum=<1 1 1>)
   )
   &InheritSourceMovement(
    Applied Position Scale=&Multiply_VectorByFloat(Vector=<1 1 1>, Float=&Lerp_Float(A=0.5, B=0.8, Alpha=$System.Timeline.NormalizedTime))
    ApplyForce=false
    ApplyPosition=true
    ApplyPositionFalloff=false
   )
   &SolveForcesAndVelocity(Clamp Velocity=false, Limit Acceleration=false)
   &ScaleSpriteSize(
    Uniform Curve Scale=0.5
    Uniform Curve Sprite Scale=DataInterface() {propertyValues {Curve {keys [
        {arriveTangent=2.9868, interpMode=RCIM_Cubic, leaveTangent=2.9868, tangentMode=RCTM_User, time=0, value=1.125}
        <at 0.4 2.3>
        {arriveTangent=-0.0405, interpMode=RCIM_Cubic, leaveTangent=-0.0405, tangentMode=RCTM_User, time=1, value=1.4}
       ]}}}
   )
   &ScaleColor(
    Scale Alpha=&FloatFromCurve(
     FloatCurve=DataInterface() {propertyValues {Curve {keys [{arriveTangent=-1.933, interpMode=RCIM_Cubic, leaveTangent=-1.933, tangentMode=RCTM_User, time=0, value=1}, <at 1 0>]}}}
     Scale Curve=&MergeHelper_FloatSwitch(0=20, 1=0.25, 2=?, 3=?, Index=$Particles.PartitionParticles.Partition, Num=::NewEnumerator0)
    )
    Scale RGB=&VectorFromFloat(Value=&FloatFromCurve(
      FloatCurve=DataInterface() {propertyValues {Curve {keys [{arriveTangent=-180.164, interpMode=RCIM_Cubic, leaveTangent=-180.164, tangentMode=RCTM_User, time=0, value=50}, <at 1 1>]}}}
      Scale Curve=&MergeHelper_FloatSwitch(0=20, 1=0.25, 2=?, 3=?, Index=$Particles.PartitionParticles.Partition, Num=::NewEnumerator0)
     ))
    ScaleA=true
    ScaleRGB=true
   )
  ]
  rendererClasses [&NiagaraSpriteRendererProperties]
  renderers [{rendererClass=&NiagaraSpriteRendererProperties}]
 }
 TrailEmbers_FE {
  emitterUpdateScript [
   &EmitterState()
   &SpawnParticlesFromOtherEmitter(
    Attribute Reader=DataInterface() {propertyValues {EmitterBinding {bindingMode=Other, emitterName=FireworkFlare_Parent_Sprite_FE}}}
    Maximum Age Spawn Threshold=15
    Minimum Age Spawn Threshold=0
    Spawn Rate=&Multiply_Float(
     A=&SelectFloatFromArray(Direct Array Index=$Engine.QualityLevel, Float Selection Array=DataInterface() {propertyValues {FloatData [10,15,15,20,20]}})
     B=1
    )
    Spawn Rate Per Particle Cap=20
    Use Maximum Age=false
    Use Minimum Age=false
   )
  ]
  particleSpawnScript [
   &PartitionParticles(Number Of Partitions=4, Partition Mode:ENiagara_PartitionMode_v2=::NewEnumerator5) {enabled=false}
   &InitializeParticle(
    Color Mode=::NewEnumerator0
    Lifetime Max=1
    Lifetime Min=0.7
    Lifetime Mode=::NewEnumerator1
    Sprite Rotation Mode=::NewEnumerator2
    Sprite Size Mode=::NewEnumerator3
    Uniform Sprite Size=&RandomRangeFloat(Minimum=0.75, Maximum=3)
    UsePositionOffset=false
    Write Material Random=true
   )
   &SampleParticlesFromOtherEmitter(
    Apply Sampled Alpha=false
    Color Sampling=::NewEnumerator0
    Emitter Sampling Mode:ENiagara_EmitterLocSamplingMode=::NewEnumerator0
    Lifetime Sampling:ENiagara_DirectReadSamplingMode=::NewEnumerator0
    Transform Data Option=1
   )
   &SphereLocation(
    MaskSpawnGroup=false
    Override Seed=false
    Sphere Radius=2
    Surface Only=false
    UseNonUniformScale=false
    UseOffset=false
    UseSphereOrientation=false
   )
  ]
  particleUpdateScript [
   &ParticleState()
   &Drag(Drag=&RandomRangeFloat(Minimum=4, Maximum=6), Use Linear Drag=true, Use Rotational Drag=false)
   &CurlNoiseForce(
    Noise Frequency=&RandomRangeFloat(Minimum=0.25, Maximum=1)
    Noise Strength=&RandomRangeFloat(Minimum=100, Maximum=200)
    Pan Noise=true
    Pan Noise Field=&RandomRangeVector(Minimum=<0 0 0>, Maximum=<1 1 1>)
   )
   &InheritSourceMovement(
    Applied Position Scale=&Multiply_VectorByFloat(Vector=<1 1 1>, Float=&Lerp_Float(A=0.5, B=0.8, Alpha=$System.Timeline.NormalizedTime))
    ApplyForce=false
    ApplyPosition=true
    ApplyPositionFalloff=false
   ) {enabled=false}
   &SolveForcesAndVelocity(Clamp Velocity=false, Limit Acceleration=false)
   &ScaleSpriteSize(
    Uniform Curve Scale=&Multiply_Float(A=0.5, B=$System.FE_SizeModifier)
    Uniform Curve Sprite Scale=DataInterface() {propertyValues {Curve {keys [
        {arriveTangent=2.9868, interpMode=RCIM_Cubic, leaveTangent=2.9868, tangentMode=RCTM_User, time=0, value=1.125}
        <at 0.4 2.3>
        {arriveTangent=-0.0405, interpMode=RCIM_Cubic, leaveTangent=-0.0405, tangentMode=RCTM_User, time=1, value=1.4}
       ]}}}
   )
   &ScaleColor(
    Scale Alpha=&FloatFromCurve(
     FloatCurve=DataInterface() {propertyValues {Curve {keys [{arriveTangent=-1.933, interpMode=RCIM_Cubic, leaveTangent=-1.933, tangentMode=RCTM_User, time=0, value=1}, <at 1 0>]}}}
     Scale Curve=&MergeHelper_FloatSwitch(0=20, 1=0.25, 2=?, 3=?, Index=$Particles.PartitionParticles.Partition, Num=::NewEnumerator0)
    )
    Scale RGB=&VectorFromFloat(Value=&FloatFromCurve(
      FloatCurve=DataInterface() {propertyValues {Curve {keys [{arriveTangent=-180.164, interpMode=RCIM_Cubic, leaveTangent=-180.164, tangentMode=RCTM_User, time=0, value=50}, <at 1 1>]}}}
      Scale Curve=&MergeHelper_FloatSwitch(0=20, 1=0.25, 2=?, 3=?, Index=$Particles.PartitionParticles.Partition, Num=::NewEnumerator0)
     ))
    ScaleA=true
    ScaleRGB=true
   )
   /SetVariables(Particles.Offset:Vector3f=$System.Offset) {isSetParameters=true}
   /SetVariables(Particles.Position=&AddVectorToPosition(Position=$Particles.Position, Vector=$Particles.Offset)) {isSetParameters=true}
   &CameraOffset(Camera Offset Amount=$System.FE_CameraOffset)
  ]
  rendererClasses [&NiagaraSpriteRendererProperties]
  renderers [{rendererClass=&NiagaraSpriteRendererProperties}]
 }
}
summary {
 systemName=NS_Contrail_TalonPane_Fireworks
 emitters [
  {emitterName=Dust, rendererClasses [&NiagaraSpriteRendererProperties]}
  {
   emitterName=Flare_With_Streak
   rendererClasses [
    &NiagaraSpriteRendererProperties
    &NiagaraLightRendererProperties
    &NiagaraMeshRendererProperties
   ]
  }
  {emitterName=Spark_High, rendererClasses [&NiagaraSpriteRendererProperties]}
  {emitterName=FireworkFlare_Parent_Sprite, rendererClasses [&NiagaraSpriteRendererProperties]}
  {emitterName=TrailEmbers, rendererClasses [&NiagaraSpriteRendererProperties]}
  {emitterName=Dust_FE, rendererClasses [&NiagaraSpriteRendererProperties]}
  {
   emitterName=Flare_With_Streak_FE
   rendererClasses [
    &NiagaraSpriteRendererProperties
    &NiagaraLightRendererProperties
    &NiagaraMeshRendererProperties
   ]
  }
  {bEnabled=false, emitterName=Spark_High_FE, rendererClasses [&NiagaraSpriteRendererProperties]}
  {emitterName=FireworkFlare_Parent_Sprite_FE, rendererClasses [&NiagaraSpriteRendererProperties]}
  {emitterName=TrailEmbers_FE, rendererClasses [&NiagaraSpriteRendererProperties]}
 ]
 userVariables [
  User.ArrowTipStartingOffset {type=Vector3f, description="", defaultValue {struct=Vector3f, value=<0 0 0>}}
  User.bIsFrontEnd {type=NiagaraBool}
  User.bIsFrontEndPreview {type=NiagaraBool}
  User.ParaGlideLeanAlpha {type=NiagaraFloat, defaultValue=0, description=""}
 ]
}
